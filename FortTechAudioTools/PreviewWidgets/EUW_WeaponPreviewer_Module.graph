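# Editor utility widget module that auditions a weapon's 1P or 3P fire sound in the level editor.
# Weapon item definitions can override sounds, fire rate, reload delay and clip size via sound libraries and stats rows.
# Timers drive initial delay, repeating shots and post-empty pauses, triggering FireSingleShot on a component spawned at the viewport camera.
# MetaSound viewmodel sliders set volume inputs, knobs drive azimuth and distance, and debug shapes mark each shot.
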EditorUtilityWidgetBlueprint /FortTechAudioTools/PreviewWidgets/EUW_WeaponPreviewer_Module
graphs [
 ApplyAssociatedItemDefinitionOverride {nodes [
   n1 = ApplyAssociatedItemDefinitionOverride!() -> (then, PlaybackDataStruct:Struct Weapon Playback Data Structure) {is=title K2Node_FunctionEntry}
   > n2 = Set LocalWeaponItemDefinitionData(LocalWeaponItemDefinitionData:Struct Weapon Playback Data Structure=@n1) -> (Output_Get:Struct Weapon Playback Data Structure)
   n3 = Break Struct Weapon Playback Data!(Struct_WeaponPlaybackData:"Struct Weapon Playback Data Structure (by ref)"=@n2) -> (
    WeaponName_17_#1:String
    3P_WeaponSound_33_#2:Sound Base Object Reference
    1P_WeaponSound_30_#3:Sound Base Object Reference
    IntialDelay_20_#4:double=~
    NumberOfRoundsPerLoop_4_#5:int
    FireRate_5_#6:double=~
    LoopCount_8_#7:int
    LoopDelay_11_#8:double=~
    AmmoCount_23_#9:int
    AssociatedItemDefinition_26_#10:Fort Weapon Ranged Item Definition Object Reference
   ) {is=title K2Node_BreakStruct}
   > n4 = Is Valid(InputObject=@n3.AssociatedItemDefinition_26_#10)
   Return Node!(execute=@n4.Is Not Valid, ItemDefinitionPlaybackData:Struct Weapon Playback Data Structure=@n2)
   > n5 = ReturnAssociatedItemDefinitionData!(execute, self, ItemDefinition:Fort Weapon Item Definition Object Reference=@n3.AssociatedItemDefinition_26_#10) -> (
    then
    FireRate:double
    ReloadTime:float
    ClipSize:int=0
    WeaponItemDefinition 3P Sound:Sound Base Object Reference
    WeaponItemDefinition1P Sound:Sound Base Object Reference
   )
   n6 = Validate or Keep Default Sounds!(
    WeaponItemDefinition3P Sound:Sound Base Object Reference=@n5.WeaponItemDefinition 3P Sound
    WeaponItemDefinition1P Sound:Sound Base Object Reference=@n5.WeaponItemDefinition1P Sound
   ) -> (3P Sound:Sound Base Object Reference, 1P Sound:Sound Base Object Reference) {is=title K2Node_MacroInstance}
   > n7 = Set members in Struct Weapon Playback Data!(
    execute:Exec
    StructRef:"Struct Weapon Playback Data Structure (by ref)"=@n2
    3P_WeaponSound_33_#2:Sound Base Object Reference=@n6
    1P_WeaponSound_30_#3:Sound Base Object Reference=@n6.1P Sound
    FireRate_5_#6:double=@n5
    LoopDelay_11_#8:double=@n5.ReloadTime
    AmmoCount_23_#9:int=@n5.ClipSize
   ) -> (then:Exec, StructOut:"Struct Weapon Playback Data Structure (by ref)") {is=title K2Node_SetFieldsInStruct}
   > Return Node(ItemDefinitionPlaybackData=@n7)
  ]}
 ApplyToggleStyle {nodes [
   n1 = ApplyToggleStyle!() -> (then, IsSelected?:bool=~, TargetSoundBorder:Border Object Reference) {is=title K2Node_FunctionEntry}
   > SetBrush(
    self:Border Object Reference=@n1.TargetSoundBorder
    InBrush=Select(
     Option 0:"Slate Brush Structure (by ref)"=Get InactiveColor() -> (InactiveColor:Slate Brush Structure)
     Option 1:"Slate Brush Structure (by ref)"=Get ActiveColor() -> (ActiveColor:Slate Brush Structure)
     Index=@n1
    ) -> (ReturnValue:"Slate Brush Structure (by ref)")
   )
   > Set RenderOpacity!(
    execute
    RenderOpacity:float=SelectFloat(A=Get InactiveOpacity() -> (InactiveOpacity:double), B=Get ActiveOpacity() -> (ActiveOpacity:double), bPickA=@n1)
    self:Widget Object Reference=@n1.TargetSoundBorder
   ) -> (then, Output_Get:float)
  ]}
 ClearAndInvalidateTimers {nodes [
   ClearAndInvalidateTimers() {is=title K2Node_FunctionEntry}
   > Clear and Invalidate Timer by Handle!(execute, WorldContextObject:Object Reference, Handle=Get FirePlaybackTimer() -> (FirePlaybackTimer:Timer Handle Structure)) -> (then)
   > Clear and Invalidate Timer by Handle(Handle=Get IntialDelayTimer() -> (IntialDelayTimer:Timer Handle Structure))
   > Clear and Invalidate Timer by Handle(Handle=Get ResumeFireTimer() -> (ResumeFireTimer:Timer Handle Structure))
  ]}
 DisplayInstanceDebugShape {nodes [
   n1 = DisplayInstanceDebugShape!() -> (then, bIsDisplayed:bool=true, SoundComponent:Audio Component Object Reference, DebugDuration:float=~, SphereRadius:float=~) {is=title K2Node_FunctionEntry}
   > Branch(Condition=@n1)
   > n2 = Set AudioComponent(AudioComponent:Audio Component Object Reference=@n1.SoundComponent) -> (Output_Get:Audio Component Object Reference)
   > Is Valid(InputObject=@n2)
   > n3 = Branch(Condition=Contains!(
     TargetMap:Map of Audio Component Object References to Linear Color Structures=Get ComponentColor() -> (ComponentColor:Map of Audio Component Object References to Linear Color Structures)
     Key:"Audio Component Object Reference (by ref)"=@n2
    ) -> (ReturnValue))
   n4 = Find(
    TargetMap:Map of Audio Component Object References to Linear Color Structures=Get ComponentColor()
    Key:"Audio Component Object Reference (by ref)"=Get AudioComponent() -> (AudioComponent:Audio Component Object Reference)
   ) -> (Value:Linear Color Structure)
   DrawDebugSphere!(
    execute=[@n8.then, @n3.then]
    WorldContextObject:Object Reference
    Center:Vector=Get World Location!(self:Scene Component Object Reference=Get AudioComponent()) -> (ReturnValue:Vector) {is=title K2Node_CallFunction}
    Radius:float=@n1.SphereRadius
    Segments:int=12
    LineColor:Linear Color Structure=@n4
    Duration:float=@n1.DebugDuration
    Thickness:float=0
    DepthPriority:EDrawDebugSceneDepthPriorityGroup Enum=Get DepthPriority() -> (DepthPriority:EDrawDebugSceneDepthPriorityGroup Enum=World)
   ) -> (then)
   n5 = Get AudioComponent()
   n6 = Get World Location(self=@n5)
   n7 = GetForwardVector!(self:Scene Component Object Reference=@n5) -> (ReturnValue:Vector)
   > DrawDebugArrow!(
    execute
    WorldContextObject:Object Reference
    LineStart:Vector=@n6
    LineEnd:Vector=vector + vector(A=@n6, B=vector * vector(A=@n7, B:float=@n1.SphereRadius) -> (ReturnValue=~))
    ArrowSize:float=20
    LineColor:Linear Color Structure=@n4
    Duration:float=@n1.DebugDuration
    Thickness:float=10
    DepthPriority:EDrawDebugSceneDepthPriorityGroup Enum=Foreground
   ) -> (then)
   > Return Node!(execute, SphereColor:Linear Color Structure=Find(TargetMap=Get ComponentColor(), Key=Get AudioComponent()))
   n8 = Add!(
    execute=@n3.else
    TargetMap:Map of Audio Component Object References to Linear Color Structures=Get ComponentColor()
    Key:"Audio Component Object Reference (by ref)"=Get AudioComponent()
    Value:"Linear Color Structure (by ref)"=HSV to RGB!(H:float=RandomFloatInRange(Min=0, Max=360), S:float=0.8, V:float=1, A:float=1) -> (ReturnValue:Linear Color Structure)
   ) -> (then) {is=title K2Node_CallFunction}
   Get DepthPriority() -> (DepthPriority=World)
  ]}
 EventGraph {nodes [
   "On Pressed (EditorUtilityButton_PlayButton)"() {is=title K2Node_ComponentBoundEvent}
   > StartPreviewingSounds()
   "On Pressed (EditorUtilityButton_StopButton)"() {is=title K2Node_ComponentBoundEvent}
   > StopPreviewer()
   n1 = "On Control Value Changed (DistanceParam)"!() -> (OutputDelegate, then, NewValue:double) {is=title K2Node_ComponentBoundEvent}
   > SetDistanceParameter!(execute, self, Param:double=@n1.NewValue) -> (then)
   n2 = "On Selected (AudioButtonToggle_PerspectiveToggle)"!() -> (OutputDelegate, then, Button:Audio Button Toggle Object Reference) {is=title K2Node_ComponentBoundEvent}
   n3 = GetIsSelected!(
    execute=[@n2.then, @n10.then]
    self:Audio Button Toggle Object Reference=Get AudioButtonToggle_PerspectiveToggle() -> (AudioButtonToggle_PerspectiveToggle:Audio Button Toggle Object Reference)
   ) -> (then, IsSelected:bool)
   > UpdateSelectedSoundToggle!(execute, self, IsSelected?:bool=@n3) -> (then)
   > UpdateMetasoundViewModel!(execute, self, TargetSound:Object Reference=GetSelectedWeaponFireSound!() -> (WeaponFireSound:Sound Base Object Reference)) -> (then)
   n4 = NOT Boolean(A=@n3)
   > Set bIsEnabled!(
    execute
    bIsEnabled:bool=@n4
    self:Widget Object Reference=Get EUW_MetaSoundFloat_Slider_MechVolume() -> (EUW_MetaSoundFloat_Slider_MechVolume:EUW Meta Sound Float Slider Default Vertical Object Reference)
   ) -> (then, Output_Get:bool)
   > Set bIsEnabled(
    bIsEnabled=@n4
    self=Get EUW_MetaSoundFloat_Slider_LowAmmoVolume() -> (EUW_MetaSoundFloat_Slider_LowAmmoVolume:EUW Meta Sound Float Slider Default Vertical Object Reference)
   )
   n5 = "On Control Value Changed (AzimuthParam)"!() -> (OutputDelegate, then, NewValue:double) {is=title K2Node_ComponentBoundEvent}
   > SetAzimuthParameter!(execute, self, Param:float=@n5.NewValue) -> (then)
   "On Pressed (EditorUtilityButton_Delete)"() {is=title K2Node_ComponentBoundEvent}
   > StopPreviewer()
   > Call On Request Close!(execute:Exec, self:EUW Weapon Previewer Module Object Reference, Module:EUW Weapon Previewer Module Object Reference=Self-Reference()) -> (then:Exec) {is=title K2Node_CallDelegate}
   "On Clicked (EditorUtilityButton_Update)"() {is=title K2Node_ComponentBoundEvent}
   n6 = Get WidgetPlaybackData() -> (WidgetPlaybackData:Struct Weapon Playback Data Structure)
   n7 = Get FortRangedWeaponItemDefinition() -> (FortRangedWeaponItemDefinition:Fort Weapon Ranged Item Definition Object Reference)
   > n8 = Branch(Condition="Equal (Object)"(
     A=@n7
     B=Break Struct Weapon Playback Data(Struct_WeaponPlaybackData=@n6) -> (
      IntialDelay_20_#4=~
      FireRate_5_#6=~
      LoopDelay_11_#8=~
     ).AssociatedItemDefinition_26_#10
    ))
   n9 = Set members in Struct Weapon Playback Data!(
    execute=@n8.else
    StructRef=@n6
    AssociatedItemDefinition_26_#10:Fort Weapon Ranged Item Definition Object Reference=@n7
   ) -> (then, StructOut)
   > UpdateWeaponPlaybackData!(execute, self, WeaponPlaybackData:Struct Weapon Playback Data Structure=@n9) -> (then)
   n10 = "On Deselected (AudioButtonToggle_PerspectiveToggle)"!() -> (OutputDelegate, then, Button:Audio Button Toggle Object Reference) {is=title K2Node_ComponentBoundEvent}
   n11 = StartFire() {is=title K2Node_CustomEvent}
   > n12 = Is Valid(InputObject=GetSelectedWeaponFireSound())
   n13 = FireSound() {is=title K2Node_CustomEvent}
   n14 = Get Value!(self:Spin Box Object Reference=Get EditorUtilitySpinBox_FireRate() -> (EditorUtilitySpinBox_FireRate:Editor Utility Spin Box Object Reference)) -> (Value:float)
   n15 = Set Timer by Event(
    execute=[@n12.Is Valid, @n22.then]
    Delegate=@n13
    Time=@n14
    bLooping=true
    bMaxOncePerFrame=true
    InitialStartDelay=float * float(A=@n14, B=-1) {is=title K2Node_PromotableOperator}
   )
   > Set FirePlaybackTimer(FirePlaybackTimer:Timer Handle Structure=@n15) -> (Output_Get:Timer Handle Structure)
   n16 = SetAzimuthParameter!() -> (OutputDelegate, then, Param:float=~) {is=title K2Node_CustomEvent}
   > Set World Rotation!(
    execute
    self:Scene Component Object Reference=Get WeaponFireComponent() -> (WeaponFireComponent:Audio Component Object Reference)
    NewRotation:Rotator=MakeRotator(Yaw=@n16.Param)
    bSweep:bool=false
    bTeleport:bool=false
   ) -> (then, SweepHitResult:Hit Result Structure) {is=title K2Node_CallFunction}
   StopPreviewer() {is=title K2Node_CustomEvent}
   > ClearAndInvalidateTimers() {is=title K2Node_CallFunction}
   n17 = Get WeaponFireComponent()
   > FadeOut!(
    execute
    self:Audio Component Object Reference=@n17
    FadeOutDuration:float=3
    FadeVolumeLevel:float=0
    FadeCurve:EAudioFaderCurve Enum=Linear
   ) -> (then)
   > Destroy Component!(execute, self:Actor Component Object Reference=@n17, Object:Object Reference) -> (then)
   > Set WeaponFireComponent(WeaponFireComponent:Audio Component Object Reference) -> (Output_Get:Audio Component Object Reference)
   StartPreviewingSounds() {is=title K2Node_CustomEvent}
   n18 = Get Value(self=Get EditorUtilitySpinBox_InitialDelay() -> (EditorUtilitySpinBox_InitialDelay:Editor Utility Spin Box Object Reference))
   > n19 = Set Timer by Event(Delegate=@n11, Time=Select(Option 0:float=@n18, Option 1:float=0.0001, Index="Equal (Float)"(A=@n18)) -> (ReturnValue:float), bLooping=false)
   > Set IntialDelayTimer(IntialDelayTimer:Timer Handle Structure=@n19) -> (Output_Get:Timer Handle Structure)
   n20 = UpdateWeaponPlaybackData!() -> (OutputDelegate, then, WeaponPlaybackData:Struct Weapon Playback Data Structure) {is=title K2Node_CustomEvent}
   > n21 = ApplyAssociatedItemDefinitionOverride!(execute, self, PlaybackDataStruct:Struct Weapon Playback Data Structure=@n20.WeaponPlaybackData) -> (then, ItemDefinitionPlaybackData:Struct Weapon Playback Data Structure) {is=title K2Node_CallFunction}
   > InitializePlaybackData!(execute, self, Struct_WeaponPlaybackData:"Struct Weapon Playback Data Structure (by ref)"=@n21) -> (then)
   > UpdateMetasoundViewModel(TargetSound=GetSelectedWeaponFireSound())
   n22 = ResumeFire() {is=title K2Node_CustomEvent}
   n23 = SetDistanceParameter!() -> (OutputDelegate, then, Param:double=~) {is=title K2Node_CustomEvent}
   n24 = Get World Location(self=Get WeaponFireComponent())
   > Set World Location!(execute, self:Scene Component Object Reference=Get WeaponFireComponent(), NewLocation:Vector=@n24, bSweep:bool=false, bTeleport:bool=false) -> (then, SweepHitResult:Hit Result Structure) {is=title K2Node_CallFunction}
   Event PreConstruct()
   > SetObject(
    self=[
     Get SinglePropertyView_WeaponFireSound_1P() -> (SinglePropertyView_WeaponFireSound_1P:Single Property View Object Reference)
     Get SinglePropertyView_WeaponFireSound_3P() -> (SinglePropertyView_WeaponFireSound_3P:Single Property View Object Reference)
     Get SinglePropertyView_WeaponItemDefinition() -> (SinglePropertyView_WeaponItemDefinition:Single Property View Object Reference)
    ]
    NewObject=Self-Reference()
   )
   Event Construct()
   > UpdateWeaponPlaybackData(WeaponPlaybackData=Get WidgetPlaybackData()) {is=title K2Node_CallFunction}
   > n25 = GetIsSelected(self=Get AudioButtonToggle_PerspectiveToggle())
   > UpdateSelectedSoundToggle(IsSelected?=@n25)
   Event Tick()
   K2Node_Composite_1 = AreAnyPlaybackTimersActive?!() -> (ReturnValue:bool=~) {is=title K2Node_Composite}
   n26 = NOT Boolean(A=@K2Node_Composite_1)
   > UpdatePlayWidgetStatus!(execute, self, bWidgetInteractable?:bool=@n26) -> (then)
   > Branch(Condition=@n26)
   > Set bIsEnabled(bIsEnabled=@n26, self=Get SinglePropertyView_WeaponFireSound_1P())
   > Set bIsEnabled(bIsEnabled=@n26, self=Get SinglePropertyView_WeaponFireSound_3P())
   Event Destruct()
   > StopPreviewer() {is=title K2Node_CallFunction}
   vector + vector(A=MakeVector(X=@n23.Param, Y=1, Z=1), B=@n24)
   n27 = Set Timer by Event(
    execute=@n33.then
    Delegate=Create Event()
    Time=Get Value(self=Get EditorUtilitySpinBox_LoopDelay() -> (EditorUtilitySpinBox_LoopDelay:Editor Utility Spin Box Object Reference))
    bLooping=false
   )
   > Set ResumeFireTimer(ResumeFireTimer:Timer Handle Structure=@n27) -> (Output_Get:Timer Handle Structure)
   n28 = GetLevelViewportCameraInfo!(WorldContextObject:Object Reference) -> (CameraLocation:Vector, CameraRotation:Rotator, ReturnValue:bool)
   n29 = SpawnSoundAtLocation!(
    execute=[@n37.else, @n36.else]
    WorldContextObject:Object Reference
    Sound:Sound Base Object Reference=Get WeaponFireSound() -> (WeaponFireSound:Sound Base Object Reference)
    Location:Vector=@n28
    Rotation:Rotator=@n28.CameraRotation
    VolumeMultiplier:float=1
    PitchMultiplier:float=1
    StartTime:float=0
    AttenuationSettings:Sound Attenuation Object Reference
    ConcurrencySettings:Sound Concurrency Object Reference
    bAutoDestroy:bool=false
   ) -> (then, ReturnValue:Audio Component Object Reference)
   > Set WeaponFireComponent(WeaponFireComponent=@n29)
   > SetAzimuthParameter(Param=Get ControlValue!(self:Audio Knob Small Object Reference=Get AzimuthParam() -> (AzimuthParam:Audio Knob Small Object Reference)) -> (ControlValue:double)) {is=title K2Node_CallFunction}
   > n30 = SetDistanceParameter(Param=Get ControlValue(self=Get DistanceParam() -> (DistanceParam:Audio Knob Small Object Reference))) {is=title K2Node_CallFunction}
   UpdateWeaponFireViewModelParameters!(execute=[@n30.then, @n37.then], self, AudioComponent:Audio Parameter Controller Interface Interface=Get WeaponFireComponent()) -> (then)
   > UpdateAmmoCount!(execute, self, DecrementValue:int=1) -> (then)
   > UpdateWeaponMetasoundInputs!(execute, self, AudioComponent:Audio Component Object Reference=Get WeaponFireComponent()) -> (then)
   n31 = Get WeaponFireComponent()
   > Execute Trigger Parameter!(execute, self:Audio Parameter Controller Interface Interface=@n31, InName:Name=FireSingleShot) -> (then)
   > DisplayInstanceDebugShape!(execute, self, bIsDisplayed:bool=true, SoundComponent:Audio Component Object Reference=@n31, DebugDuration:float=0.2, SphereRadius:float=600) -> (then, SphereColor:Linear Color Structure) {is=title K2Node_CallFunction}
   n32 = Get Value(self=Get EditorUtilitySpinBox_AmmoCount() -> (EditorUtilitySpinBox_AmmoCount:Editor Utility Spin Box Object Reference))
   > Branch(Condition="Equal (Integer)"(B=Truncate(A=@n32)))
   > Pause Timer by Handle!(execute, WorldContextObject:Object Reference, Handle:Timer Handle Structure=Get FirePlaybackTimer()) -> (then)
   > n33 = UpdateAmmoCount(DecrementValue=1)
   n34 = GetSelectedWeaponFireSound()
   n35 = Get WeaponFireComponent()
   n36 = Branch(
    execute=@n13.then
    Condition="Equal (Object)"(A=Get Sound!(self:Audio Component Object Reference=@n35) -> (Sound:Sound Base Object Reference), B=@n34)
   )
   > n37 = Get!(execute:Exec) -> (then:Exec, else:Exec, WeaponFireComponent:Audio Component Object Reference)
  ]}
 AreAnyPlaybackTimersActive? {
  path=EventGraph.K2Node_Composite_1.AreAnyPlaybackTimersActive?
  nodes [
   Inputs!()
   n1 = Get IntialDelayTimer()
   n2 = Get ResumeFireTimer()
   n3 = Get FirePlaybackTimer()
   Outputs!(ReturnValue:bool=OR Boolean!(
     A=Does Timer Exist by Handle!(WorldContextObject:Object Reference, Handle:Timer Handle Structure=@n1) -> (ReturnValue:bool)
     B=Does Timer Exist by Handle(Handle=@n3)
     C:bool=Does Timer Exist by Handle(Handle=@n2)
    ) -> (ReturnValue))
  ]
 }
 Get Input Viewmodels {nodes [
   n1 = Get Input Viewmodels!() -> (then, Input Names:Array of Names, MetaSound Viewmodel:Meta Sound Viewmodel Object Reference) {is=title K2Node_FunctionEntry}
   > n2 = For Each Loop(Array:Array of Names=@n1) -> (Array Element:Name)
   Return Node!(
    execute=@n2.Completed
    Input Viewmodels:Map of Names to Meta Sound Input Viewmodel Object References=Get ViewModels() -> (ViewModels:Map of Names to Meta Sound Input Viewmodel Object References)
   )
   > Add(
    TargetMap:Map of Names to Meta Sound Input Viewmodel Object References=Get ViewModels()
    Key:"Name (by ref)"=@n2
    Value:"Meta Sound Input Viewmodel Object Reference (by ref)"=Find Input Viewmodel!(self:Meta Sound Viewmodel Object Reference=@n1.MetaSound Viewmodel, InputViewModelName:Name=@n2) -> (ReturnValue:Meta Sound Input Viewmodel Object Reference)
   )
  ]}
 GetSelectedWeaponFireSound {nodes [
   GetSelectedWeaponFireSound() {is=title K2Node_FunctionEntry}
   > n1 = GetIsSelected(self=Get AudioButtonToggle_PerspectiveToggle())
   > n2 = Set WeaponFireSound(WeaponFireSound:Sound Base Object Reference=Select(
     Option 0:Sound Base Object Reference=Get 1PWeaponSound() -> (1PWeaponSound:Sound Base Object Reference)
     Option 1:Sound Base Object Reference=Get 3PWeaponSound() -> (3PWeaponSound:Sound Base Object Reference)
     Index=@n1
    ) -> (ReturnValue:Sound Base Object Reference)) -> (Output_Get:Sound Base Object Reference)
   > Return Node!(execute, WeaponFireSound:Sound Base Object Reference=@n2)
  ]}
 InitializePlaybackData {nodes [
   n1 = InitializePlaybackData!() -> (then, Struct_WeaponPlaybackData:"Struct Weapon Playback Data Structure (by ref)") {is=title K2Node_FunctionEntry}
   n2 = Break Struct Weapon Playback Data(Struct_WeaponPlaybackData=@n1) -> (
    IntialDelay_20_#4=~
    FireRate_5_#6=~
    LoopDelay_11_#8=~
   )
   > Set FortRangedWeaponItemDefinition(FortRangedWeaponItemDefinition:Fort Weapon Ranged Item Definition Object Reference=@n2.AssociatedItemDefinition_26_#10) -> (Output_Get:Fort Weapon Ranged Item Definition Object Reference)
   > Set 3PWeaponSound(3PWeaponSound:Sound Base Object Reference=@n2.3P_WeaponSound_33_#2) -> (Output_Get:Sound Base Object Reference)
   > Set 1PWeaponSound(1PWeaponSound:Sound Base Object Reference=@n2.1P_WeaponSound_30_#3) -> (Output_Get:Sound Base Object Reference)
   > Set Value!(execute, Value:float=@n2.FireRate_5_#6, self:Spin Box Object Reference=Get EditorUtilitySpinBox_FireRate()) -> (then, Output_Get:float)
   > Set Value(Value=@n2.LoopDelay_11_#8, self=Get EditorUtilitySpinBox_LoopDelay())
   > Set Value(Value=@n2.IntialDelay_20_#4, self=Get EditorUtilitySpinBox_InitialDelay())
   n3 = "To Float (Integer)"(InInt=@n2.AmmoCount_23_#9)
   n4 = Get EditorUtilitySpinBox_AmmoCount()
   > Set MaxValue!(execute, MaxValue:float=@n3, self:Spin Box Object Reference=@n4) -> (then, Output_Get:float)
   > Set MaxSliderValue!(execute, MaxSliderValue:float=@n3, self:Spin Box Object Reference=@n4) -> (then, Output_Get:float)
   > Set Value(Value=@n3, self=@n4)
  ]}
 OnRequestClose {nodes [OnRequestClose!() -> (then, Module:EUW Weapon Previewer Module Object Reference) {is=title K2Node_FunctionEntry}]}
 ReturnAssociatedItemDefinitionData {nodes [
   n1 = ReturnAssociatedItemDefinitionData!() -> (then, ItemDefinition:Fort Weapon Item Definition Object Reference) {is=title K2Node_FunctionEntry}
   > n2 = Sequence()
   > n3 = GetEditorProperty!(execute:Exec, Object:Object Reference=@n1, PropertyName:Name=DataList) -> (then:Exec, PropertyValue:Array of Instanced Struct Structures, ReturnValue:bool)
   > Branch(Condition=@n3.ReturnValue)
   > n4 = For Each Loop(Array:Array of Instanced Struct Structures=@n3) -> (Array Element:Instanced Struct Structure)
   > n5 = GetInstancedStructValue!(execute:Exec, InstancedStruct:"Instanced Struct Structure (by ref)"=@n4) -> (Valid:Exec, NotValid:Exec, Value:"Fortnite Weapon Sound Libraries Item Data Structure (by ref)") {is=title K2Node_InstancedStruct}
   > n6 = For Each Loop(Array:Array of Sound Library Object References=Break Fortnite Weapon Sound Libraries Item Data!(FortItemComponentData_WeaponSoundLibraries:"Fortnite Weapon Sound Libraries Item Data Structure (by ref)"=@n5) -> (
     WeaponSoundLibraries:Array of Sound Library Object References
     WeaponOwnerSoundLibraries:Array of Sound Library Object References
     WeaponCollectionTag:Gameplay Tag Container Structure
    ) {is=title K2Node_BreakStruct}) -> (Array Element:Sound Library Object Reference)
   > n7 = "For Each Loop (Map)"!(
    execute:Exec
    MapPin:Map of Gameplay Tag Structures to Sound Base Soft Object References=Get Sounds!(self:Sound Library Object Reference=@n6) -> (Sounds:Map of Gameplay Tag Structures to Sound Base Soft Object References)
    BreakPin:Exec
   ) -> (then:Exec, KeyPin:Gameplay Tag Structure, ValuePin:Sound Base Soft Object Reference, CompletedPin:Exec)
   > n8 = Switch on Gameplay Tag!(execute:Exec, Selection:Gameplay Tag Structure=@n7) -> (Default:Exec, SoundLibrary.Weapon.Fire.1P:Exec, SoundLibrary.Weapon.Fire.3P:Exec) {is=title GameplayTagsK2Node_SwitchGameplayTag, pinTags [{tagName=SoundLibrary.Weapon.Fire.1P}, {tagName=SoundLibrary.Weapon.Fire.3P}]}
   n9 = LoadAsset_Blocking(execute=@n8.SoundLibrary.Weapon.Fire.1P, Asset=@n7.ValuePin)
   n10 = Cast To SoundBase!(Object=@n9) -> (AsSound Base:Sound Base Object Reference, bSuccess:bool=~) {is=title K2Node_DynamicCast}
   > Branch(Condition=@n10.bSuccess)
   > Set WeaponItemDefinitionDefined1P Sound(WeaponItemDefinitionDefined1P Sound:Sound Base Object Reference=@n10) -> (Output_Get:Sound Base Object Reference)
   n11 = LoadAsset_Blocking(execute=@n8.SoundLibrary.Weapon.Fire.3P, Asset=@n7.ValuePin)
   n12 = Cast To SoundBase(Object=@n11) -> (bSuccess=~)
   > Branch(Condition=@n12.bSuccess)
   > Set WeaponItemDefinition Defined 3P Sound(WeaponItemDefinition Defined 3P Sound:Sound Base Object Reference=@n12) -> (Output_Get:Sound Base Object Reference)
   n13 = GetWeaponStatHandle!(self:Fort Weapon Item Definition Object Reference=@n1) -> (ReturnValue_DataTable:Data Table Object Reference, ReturnValue_RowName:Name)
   n14 = Get Data Table Row!(execute:Exec=@n2.then_1, DataTable:Data Table Object Reference=@n13, RowName:Name=@n13.ReturnValue_RowName) -> (then:Exec, RowNotFound:Exec, ReturnValue:Fort Ranged Weapon Stats Structure)
   n15 = Break Fort Ranged Weapon Stats!(FortRangedWeaponStats:"Fort Ranged Weapon Stats Structure (by ref)"=@n14) -> (
    Spread:float=~
    OverrideBaseSpread:float=~
    SpreadDownsights:float=~
    JumpingFallingSpreadDownsights:float=~
    OverrideSpreadDownsights:float=~
    StandingStillSpreadMultiplier:float=~
    AthenaCrouchingSpreadMultiplier:float=~
    AthenaSlidingSpreadMultiplier:float=~
    AthenaJumpingFallingSpreadMultiplier:float=~
    AthenaSprintingSpreadMultiplier:float=~
    MinSpeedForSpreadMultiplier:float=~
    MaxSpeedForSpreadMultiplier:float=~
    SpreadDownsightsAdditionalCooldownTime:float=~
    SpreadExponent:float=~
    HeatX1:float=~
    HeatY1:float=~
    HeatX2:float=~
    HeatY2:float=~
    HeatX3:float=~
    HeatY3:float=~
    HeatXScale:float=~
    HeatYScale:float=~
    CoolX1:float=~
    CoolY1:float=~
    CoolX2:float=~
    CoolY2:float=~
    CoolX3:float=~
    CoolY3:float=~
    CoolXScale:float=~
    CoolYScale:float=~
    PerfectAimCooldown:float=~
    bUseConstantCooldownSpeed:bool=~
    ConstantCooldownSpeed:float=~
    bUseConstantCooldownRecoverySpeed:bool=~
    ConstantCooldownRecoverySpeed:float=~
    bUseMovementCooldownOverride:bool=~
    BulletsPerCartridge:int
    MinHitsPerCartridge:int
    FiringRate:float=~
    ROFScale:float=~
    BurstFiringRate:float=~
    FiringRateDownsightsMultiplier:float=~
    EarlyRefireRate:float=~
    EarlyRefireSpreadPenalty:float=~
    EarlyRefireSpreadMinimumMultiplier:float=~
    EarlyRefireSpreadDownsightsMinimumMultiplier:float=~
    RecoilVert:float=~
    RecoilVertScale:float=~
    RecoilVertScaleGamepad:float=~
    VertRecoilDownChance:float=~
    RecoilHoriz:float=~
    RecoilHorizScale:float=~
    RecoilHorizScaleGamepad:float=~
    RecoilInterpSpeed:float=~
    RecoilRecoveryInterpSpeed:float=~
    RecoilRecoveryDelay:float=~
    RecoilRecoveryEaseInTime:float=~
    RecoilRecoveryFraction:float=~
    RecoilDownsightsMultiplier:float=~
    RecoilStandingStillMultiplier:float=~
    RecoilMovingMultiplier:float=~
    AthenaRecoilMagnitudeMin:float=~
    AthenaRecoilMagnitudeMax:float=~
    AthenaRecoilMagnitudeScale:float=~
    AthenaRecoilAngleMin:float=~
    AthenaRecoilAngleMax:float=~
    AthenaRecoilInterpSpeed:float=~
    AthenaRecoilRecoveryInterpSpeed:float=~
    AthenaRecoilDownsightsMultiplier:float=~
    AthenaRecoilHipFireMultiplier:float=~
    PitchLimitForPerBulletRecoil:float=~
    bPerBulletRecoilRecoveryWhileFiring:bool=~
    bEvenRecoilWhenCountered:bool=~
    bUseOffsetRecoil:bool=~
    OffsetRecoilFactor:float=~
    RecoilBounceFactor:float=~
    RecoilBounceOffsetFactor:float=~
    OffsetRecoilLimitOverride:float=~
    AthenaAimAssistRange:float=~
    ADSTransitionInTime:float=~
    ADSTransitionOutTime:float=~
    ToggleScopedViewTransitionTime:float=~
    MaxSpareAmmo:int
    BulletsPerTracer:int
    AIDelayBeforeFiringMin:float=~
    AIDelayBeforeFiringMax:float=~
    AIFireDurationMin:float=~
    AIFireDurationMax:float=~
    AIMinSpreadDuration:float=~
    AIMaxSpreadDuration:float=~
    AIDurationSpreadMultiplier:float=~
    AIAdditionalSpreadForTargetMovingLaterally:float=~
    AIAthenaHearFiringNoiseRange:Scalable Float Structure
    AIAthenaHearLightweightProjectileImpactNoiseRange:Scalable Float Structure
    AIAthenaHearLightweightProjectileFlyByNoiseRange:Scalable Float Structure
    AIAthenaLightweightProjectileFlyByEmissionInterval:Scalable Float Structure
    EQSDensity:float=~
    MinApproachRange:float=~
    MinActualRange:float=~
    MinPreferredRange:float=~
    MinPreferredRangeEQS:float=~
    MaxPreferredRangeEQS:float=~
    MaxPreferredRange:float=~
    MaxActualRange:float=~
    MaxApproachRange:float=~
    RangeToAutomaticallyAddEnemyPawnGoals:float=~
    SweepRadius:float=~
    AutoReloadDelayOverride:float=~
    OverheatingMaxValue:float=~
    OverheatHeatingValue:float=~
    FullChargeOverheatHeatingValue:float=~
    OverheatingCoolingValue:float=~
    FullyOverheatedCoolingValue:float=~
    HeatingCooldownDelay:float=~
    OverheatedCooldownDelay:float=~
    bCoolOverheatWhileCharging:bool=~
    bCoolOverheatWhileUnequipped:bool=~
    ManualVentingMinOverheatValueRequirement:float=~
    FortHomingTurnSpeedMin:float=~
    FortHomingTurnSpeedMax:float=~
    FortHomingTimeUntilMaxTurnSpeed:float=~
    ProjectileVelocity:float=~
    BulletsPerBulletFireFX:int
    MaxDamagePerCartridge:float=~
    bUseConstantLerpForRecoil:bool=~
    bStartRecoilDelayAfterRecoilPeaks:bool=~
    bUseAccumulatedRecoilToSelectShotIndex:bool=~
    RecoilFractionToBulletCountCurve:Curve Float Soft Object Reference
    bUseAccumulatedRecoilToLerpMinToMaxPerBulletRecoilValues:bool=~
    RecoilMinToMaxCurve:Curve Float Soft Object Reference
    RecoilMinVerticalCurve:Curve Float Soft Object Reference
    RecoilMaxVerticalCurve:Curve Float Soft Object Reference
    RecoilMinHorizontalCurve:Curve Float Soft Object Reference
    RecoilMaxHorizontalCurve:Curve Float Soft Object Reference
    RecoilErrorAddedPerShot:float=~
    RecoilOffsetConsideredMaxFraction:float=~
    RecoilDelay:float=~
    bDelaySpreadWithRecoil:bool=~
    bDisableOneFrameFiringAnimationDelay:bool=~
    bUseScopedPerBulletRecoilOverridesForBasicADS:bool=~
    bLerpSprintingSpreadMultiplierWithMovementSpeed:bool=~
    SpreadMultiplierCap:float=~
    PerBulletRecoil:Array of Fort Ranged Weapon Recoil Structures
    ScopedPerBulletRecoilOverride:Array of Fort Ranged Weapon Recoil Structures
    PerBulletSpread:Array of Fort Ranged Weapon Spread Structures
    OverrideBulletPattern:Array of Fort Bullet Pattern Entry Data Structures
    BaseLevel:int
    NamedWeightRow:Name
    DmgPB:float=~
    DmgMid:float=~
    DmgLong:float=~
    DmgMaxRange:float=~
    EnvDmgPB:float=~
    EnvDmgMid:float=~
    EnvDmgLong:float=~
    EnvDmgMaxRange:float=~
    ImpactDmgPB:float=~
    ImpactDmgMid:float=~
    ImpactDmgLong:float=~
    ImpactDmgMaxRange:float=~
    bForceControl:bool=~
    RngPB:float=~
    RngMid:float=~
    RngLong:float=~
    RngMax:float=~
    DmgScaleTable:Curve Table Soft Object Reference
    DmgScaleTableRow:Name
    DmgScale:float=~
    EnvDmgScaleTable:Curve Table Soft Object Reference
    EnvDmgScaleTableRow:Name
    EnvDmgScale:float=~
    ImpactDmgScaleTable:Curve Table Soft Object Reference
    ImpactDmgScaleTableRow:Name
    ImpactDmgScale:float=~
    SurfaceRatioRowName:Name
    DamageZone_Light:float=~
    DamageZone_Normal:float=~
    DamageZone_Critical:float=~
    DamageZone_Vulnerability:float=~
    DamageZone_Invulnerability:float=~
    KnockbackMagnitude:float=~
    MidRangeKnockbackMagnitude:float=~
    LongRangeKnockbackMagnitude:float=~
    KnockbackZAngle:float=~
    ShortRangeHitImpulseMagnitude:float=~
    MidRangeHitImpulseMagnitude:float=~
    LongRangeHitImpulseMagnitude:float=~
    HitImpulseZBias:float=~
    StunTime:float=~
    StunScale:float=~
    Durability:Data Table Soft Object Reference
    DurabilityRowName:Name
    DurabilityScale:float=~
    DurabilityPerUse:float=~
    DurabilityCooldown:Data Table Soft Object Reference
    DurabilityCooldownRowName:Name
    DurabilityCooldownScale:float=~
    FullChargeDurabilityPerUse:float=~
    DiceCritChance:float=~
    DiceCritDamageMultiplier:float=~
    ReloadTime:float=~
    ReloadScale:float=~
    ReloadType:EFortWeaponReloadType Enum
    bAllowReloadInterrupt:bool=~
    bReloadInterruptIsImmediate:bool=~
    NumIndividualBulletsToReload:int
    ClipSize:int
    ClipScale:float=~
    InitialClips:int
    bShouldAddInitialAmmo:bool=~
    CartridgePerFire:int
    AmmoCostPerFire:int
    MaxAmmoCostPerFire:int
    MinChargeTime:float=~
    MaxChargeTime:float=~
    ChargeDownTime:float=~
    bAutoDischarge:bool=~
    MaxChargeTimeUntilDischarge:float=~
    MinChargeDamageMultiplier:float=~
    MaxChargeDamageMultiplier:float=~
    ChargeDamageMultiplierCurve:Curve Float Soft Object Reference
    EquipAnimRate:float=~
    StowAnimRate:float=~
    QuickBarSlotCooldownDuration:float=~
    AimAssistPullStrengthMultiplier:float=~
    AimAssistSlowStrengthMultiplier:float=~
    AimAssistAutoAimSize:float=~
    MeleeAimAssistRequiredTag:Gameplay Tag Structure
    MaxUnderReticleDistance:float=~
    AutofireRange:float=~
    AutofireMinRange:float=~
    AutofireZoomTargetMinSizePercent:float=~
    AutofireAcquisitionDelay:float=~
    AutofireDBNOAcquisitionDelay:float=~
    AutofireAcquisitionRechargeTime:float=~
    AutofireMinChargingTime:float=~
    AutofireReleaseTime:float=~
    AutofireCooldown:float=~
    PerWeaponHotfixes:Fort Per Weapon Hotfixes Structure
   ) {is=title K2Node_BreakStruct}
   Return Node!(
    execute=[@n14.RowNotFound, @n14.then]
    FireRate:double=SafeDivide(A=1, B=@n15.FiringRate)
    ReloadTime:float=@n15.ReloadTime
    ClipSize:int=@n15.ClipSize
    WeaponItemDefinition 3P Sound:Sound Base Object Reference=Get WeaponItemDefinition Defined 3P Sound() -> (WeaponItemDefinition Defined 3P Sound:Sound Base Object Reference)
    WeaponItemDefinition1P Sound:Sound Base Object Reference=Get WeaponItemDefinitionDefined1P Sound() -> (WeaponItemDefinitionDefined1P Sound:Sound Base Object Reference)
   )
  ]}
 TraceFromViewportCamera {nodes [
   n1 = TraceFromViewportCamera!() -> (then, TraceLength:int=2000) {is=title K2Node_FunctionEntry}
   n2 = GetLevelViewportCameraInfo()
   n3 = vector * vector(
    A=Rotate Vector!(A:Vector=GetForwardVector!(InRot:Rotator=@n2.CameraRotation) -> (ReturnValue), B:Rotator="0, 0, 0") -> (ReturnValue:Vector)
    B=2
   ) -> (ReturnValue=~)
   > n4 = Line Trace By Channel!(
    execute
    WorldContextObject:Object Reference
    Start:Vector=@n2
    End:Vector=vector + vector(A=@n2, B=vector * vector(A=@n3, B:int=@n1) -> (ReturnValue=~))
    TraceChannel:ETraceTypeQuery Enum
    bTraceComplex:bool=true
    ActorsToIgnore:Array of Actor Object References
    DrawDebugType:EDrawDebugTrace Enum=ForDuration
    bIgnoreSelf:bool
    TraceColor:Linear Color Structure=(R=0.7917,G=0.0558,B=0.6967,A=1)
    TraceHitColor:Linear Color Structure
    DrawTime:float=1
   ) -> (then, OutHit:Hit Result Structure, ReturnValue:bool)
   > n5 = Branch(Condition=@n4.ReturnValue)
   > Return Node!(execute, OutHit:Hit Result Structure=@n4)
   RaiseScriptError!(execute=@n5.else, ErrorMessage:String=Line Trace By Channel Failed) -> (then)
  ]}
 UpdateAmmoCount {nodes [
   n1 = UpdateAmmoCount!() -> (then, DecrementValue:int=1) {is=title K2Node_FunctionEntry}
   > Branch(Condition=IsChecked(self=Get EditorUtilityCheckBox_bSimulateAmmo() -> (EditorUtilityCheckBox_bSimulateAmmo:Editor Utility Check Box Object Reference)))
   n2 = Get EditorUtilitySpinBox_AmmoCount()
   n3 = Get Value(self=@n2)
   > SetValue!(
    execute
    self:Spin Box Object Reference=Get EditorUtilitySpinBox_AmmoCount()
    NewValue:float=SelectFloat(
     A=GetMaxValue!(self:Spin Box Object Reference=@n2) -> (ReturnValue:float)
     B=float - float(A=@n3, B:int=@n1)
     bPickA="float < float"(A=@n3, B=1)
    )
   ) -> (then)
  ]}
 UpdateMetasoundViewModel {nodes [
   n1 = UpdateMetasoundViewModel!() -> (then, TargetSound:Object Reference) {is=title K2Node_FunctionEntry}
   > n2 = Cast To MetaSoundSource!(execute, Object=@n1) -> (then, CastFailed, AsMetaSound Source:Meta Sound Source Object Reference) {is=title K2Node_DynamicCast}
   n3 = Get MetaSoundEditorViewModel() -> (MetaSoundEditorViewModel:Meta Sound Editor Viewmodel Object Reference)
   > Initialize MetaSound!(execute, self:Meta Sound Viewmodel Object Reference=@n3, InMetaSound:Meta Sound Document Interface Interface=@n2) -> (then)
   > n4 = For Each Loop(Array:Array of Meta Sound Literal Widget Interface Interfaces=Make Array!(
     "[0]":Meta Sound Literal Widget Interface Interface=Get EUW_MetaSoundFloat_Slider_MasterVolume() -> (EUW_MetaSoundFloat_Slider_MasterVolume:EUW Meta Sound Float Slider Default Vertical Object Reference)
     "[1]":Meta Sound Literal Widget Interface Interface=Get EUW_MetaSoundFloat_Slider_MechVolume()
     "[2]":Meta Sound Literal Widget Interface Interface=Get EUW_MetaSoundFloat_Slider_LowAmmoVolume()
    ) -> (Array:Array of Meta Sound Literal Widget Interface Interfaces)) -> (Array Element:Meta Sound Literal Widget Interface Interface)
   > Set Input Viewmodels!(
    execute
    self:Meta Sound Literal Widget Interface Interface=@n4
    InputViewModels:Map of Names to Meta Sound Input Viewmodel Object References=Get Input Viewmodels!(
     Input Names:Array of Names=Get Input Viewmodel Names!(self:Meta Sound Literal Widget Interface Interface=@n4) -> (ReturnValue:Array of Names) {is=title K2Node_CallFunction}
     MetaSound Viewmodel:Meta Sound Viewmodel Object Reference=@n3
    ) -> (Input Viewmodels:Map of Names to Meta Sound Input Viewmodel Object References) {is=title K2Node_CallFunction}
   ) -> (then) {is=title K2Node_CallFunction}
  ]}
 UpdatePlayWidgetStatus {nodes [
   n1 = UpdatePlayWidgetStatus!() -> (then, bWidgetInteractable?:bool=~) {is=title K2Node_FunctionEntry}
   > Set Text(
    Text=SelectText(A="LOCTEXT(\"Play\")", B="LOCTEXT(\"Playing\")", bPickA=@n1)
    self=Get TextBlock_Play() -> (TextBlock_Play:Text Object Reference)
   )
   n2 = Get EditorUtilityButton_PlayButton() -> (EditorUtilityButton_PlayButton:Editor Utility Button Object Reference)
   > Set bIsEnabled(bIsEnabled=@n1, self=@n2)
   > SetColorAndOpacity(
    self:Button Object Reference=@n2
    InColorAndOpacity:Linear Color Structure=SelectColor(A=(R=1,G=1,B=1,A=1), B=(R=1,G=0.2576,B=0.1883,A=0.513), bPickA=@n1)
   )
  ]}
 UpdateSelectedSoundToggle {nodes [
   n1 = UpdateSelectedSoundToggle!() -> (then, IsSelected?:bool=~) {is=title K2Node_FunctionEntry}
   > ApplyToggleStyle!(
    execute
    self
    IsSelected?:bool=NOT Boolean(A=@n1)
    TargetSoundBorder:Border Object Reference=Get Border_3P() -> (Border_3P:Border Object Reference)
   ) -> (then) {is=title K2Node_CallFunction}
   > ApplyToggleStyle(IsSelected?=@n1, TargetSoundBorder=Get Border_1P() -> (Border_1P:Border Object Reference))
  ]}
 UpdateWeaponFireViewModelParameters {nodes [
   n1 = UpdateWeaponFireViewModelParameters!() -> (then, AudioComponent:Audio Parameter Controller Interface Interface) {is=title K2Node_FunctionEntry}
   n2 = Get EUW_MetaSoundFloat_Slider_MasterVolume()
   > Set Float Parameter!(
    execute
    self:Audio Parameter Controller Interface Interface=@n1
    InName:Name=MasterVolume
    InFloat:float=Get SourceValue!(self:Meta Sound Literal Float Viewmodel Object Reference=Get MetaSoundLiteralViewModel_Float!(self:EUW Meta Sound Float Slider Default Vertical Object Reference=@n2) -> (MetaSoundLiteralViewModel_Float:Meta Sound Literal Float Viewmodel Object Reference)) -> (SourceValue:float)
   ) -> (then) {is=title K2Node_CallFunction}
   n3 = Get EUW_MetaSoundFloat_Slider_LowAmmoVolume()
   > Set Float Parameter(self=@n1, InName=LowAmmo.Volume, InFloat=Get SourceValue(self=Get MetaSoundLiteralViewModel_Float(self=@n3)))
   n4 = Get EUW_MetaSoundFloat_Slider_MechVolume()
   > Set Float Parameter(self=@n1, InName=Mech.Volume, InFloat=Get SourceValue(self=Get MetaSoundLiteralViewModel_Float(self=@n4)))
   > Return Node!(execute)
  ]}
 UpdateWeaponMetasoundInputs {nodes [
   n1 = UpdateWeaponMetasoundInputs!() -> (then, AudioComponent:Audio Component Object Reference) {is=title K2Node_FunctionEntry}
   > Set Boolean Parameter!(execute, self:Audio Parameter Controller Interface Interface=@n1, InName:Name=bHasUnlimitedAmmo, InBool:bool=true) -> (then) {is=title K2Node_CallFunction}
   > n2 = GetIsSelected(self=Get AudioButtonToggle_AimDownSights() -> (AudioButtonToggle_AimDownSights:Audio Button Toggle Object Reference))
   > Set Boolean Parameter(self=@n1, InName=Weapon.IsAimingDownSights, InBool=@n2)
   n3 = Get EditorUtilitySpinBox_AmmoCount()
   > Set Integer Parameter!(
    execute
    self:Audio Parameter Controller Interface Interface=@n1
    InName:Name=Weapon.AmmoRemaining
    InInt:int=Truncate(A=Get Value(self=@n3))
   ) -> (then) {is=title K2Node_CallFunction}
  ]}
 ValidateOrKeepDefaultSounds {nodes [
   n1 = Get LocalWeaponItemDefinitionData() -> (LocalWeaponItemDefinitionData:Struct Weapon Playback Data Structure)
   n2 = Inputs!() -> (WeaponItemDefinition3P Sound:Sound Base Object Reference, WeaponItemDefinition1P Sound:Sound Base Object Reference)
   Outputs!(
    3P Sound:Sound Base Object Reference=Select(
     Option 0=Break Struct Weapon Playback Data(Struct_WeaponPlaybackData=@n1) -> (
      IntialDelay_20_#4=~
      FireRate_5_#6=~
      LoopDelay_11_#8=~
     ).3P_WeaponSound_33_#2
     Option 1=@n2
     Index=IsValid(Object=@n2)
    )
    1P Sound:Sound Base Object Reference=Select(
     Option 0=Break Struct Weapon Playback Data(Struct_WeaponPlaybackData=@n1) -> (
      IntialDelay_20_#4=~
      FireRate_5_#6=~
      LoopDelay_11_#8=~
     ).1P_WeaponSound_30_#3
     Option 1=@n2.WeaponItemDefinition1P Sound
     Index=IsValid(Object=@n2.WeaponItemDefinition1P Sound)
    )
   )
  ]}
]
variables [
 1PWeaponSound {type=SoundBase, default=/Game/QA/Art_QA/BR_Audio/Audio/MSS_BR_Audio_Paprika_SCAR_Fire_1p}
 3PWeaponSound {type=SoundBase, default=/Game/QA/Art_QA/BR_Audio/Audio/MSS_BR_Audio_Paprika_SCAR_Fire_3p}
 ActiveColor {
  type=SlateBrush
  default {
   bIsDynamicallyLoaded=false
   drawAs=RoundedBox
   imageType=NoImage
   mirroring=NoMirror
   resourceName=None
   resourceObject=null
   tiling=NoTile
   imageSize=<32 32>
   margin=<ltrb 0 0 0 0>
   outlineSettings {
    bUseBrushTransparency=false
    roundingType=HalfHeightRadius
    width=0
    color {colorUseRule=UseColor_Specified, specifiedColor=<rgba 0 0 0 0>}
    cornerRadii=<0 0 0 0>
   }
   tintColor {colorUseRule=UseColor_Specified, specifiedColor=<rgba 0 0 0 1>}
   uVRegion {bIsValid=false, max=<0 0>, min=<0 0>}
  }
 }
 ActiveOpacity {type=number, default=1}
 bFinishedLoopSeqeuence {type=boolean}
 Camera Location {type=Vector}
 ComponentColor {default {}}
 DefaultControlBusMix {type=SoundControlBusMix, default=/Game/Sounds/Modulation/Mixes/CBM_DuckOnResultsScreen}
 DistanceScalar {type=number, default=1.1}
 Duration {type=number, default=0.2}
 FirePlaybackTimer {type=TimerHandle}
 FortRangedWeaponItemDefinition {type=FortWeaponRangedItemDefinition, default=null}
 InactiveColor {
  type=SlateBrush
  default {
   bIsDynamicallyLoaded=false
   drawAs=RoundedBox
   imageType=NoImage
   mirroring=NoMirror
   resourceName=None
   resourceObject=null
   tiling=NoTile
   imageSize=<32 32>
   margin=<ltrb 0 0 0 0>
   outlineSettings {
    bUseBrushTransparency=false
    roundingType=HalfHeightRadius
    width=0
    color {colorUseRule=UseColor_Specified, specifiedColor=<rgba 0 0 0 0>}
    cornerRadii=<0 0 0 0>
   }
   tintColor {colorUseRule=UseColor_Specified, specifiedColor=<rgba 0.3478 1 0.4836 0.3217>}
   uVRegion {bIsValid=false, max=<0 0>, min=<0 0>}
  }
 }
 InactiveOpacity {type=number, default=0.2}
 IntialDelayTimer {type=TimerHandle}
 LastSoundLocation {type=Vector}
 LoopPlaybackTimer {type=TimerHandle}
 NewVar {type=FortWeaponSoundLibraryComponent, default=null}
 NumberOfTimesLooped {type=integer}
 NumberOfTimesPlayed {type=integer}
 OnRequestClose {default=()}
 ResumeFireTimer {type=TimerHandle}
 Selected Color {
  type=SlateBrush
  default {
   bIsDynamicallyLoaded=false
   drawAs=Image
   imageType=NoImage
   mirroring=NoMirror
   resourceName=None
   resourceObject=null
   tiling=NoTile
   imageSize=<32 32>
   margin=<ltrb 0 0 0 0>
   outlineSettings {
    bUseBrushTransparency=false
    roundingType=HalfHeightRadius
    width=0
    color {colorUseRule=UseColor_Specified, specifiedColor=<rgba 0 0 0 0>}
    cornerRadii=<0 0 0 0>
   }
   tintColor {colorUseRule=UseColor_Specified, specifiedColor=<rgba 1 1 1 1>}
   uVRegion {bIsValid=false, max=<0 0>, min=<0 0>}
  }
 }
 soundlib {type=SoundLibrary, default=null}
 Source Value {type=number}
 True {type=SoundBase, default=null}
 ViewportDebugText {type=TextRenderActor, default=null}
 WeaponFireComponent {type=AudioComponent, default=null}
 WeaponFireSound {type=SoundBase, default=/PaprikaCoreWeapons/Items/Weapons/PaprikaAR_DPS/Sounds/MSS_Paprika_SCAR_Fire_1p}
 WidgetPlaybackData {
  type=Struct_WeaponPlaybackData
  default {
   1P_WeaponSound=&MSS_BR_Audio_Paprika_SCAR_Fire_1p
   3P_WeaponSound=&MSS_BR_Audio_Paprika_SCAR_Fire_3p
   ammoCount=30
   associatedItemDefinition=null
   fireRate=0.1
   intialDelay=0
   loop Count=2
   loop Delay=1
   numberOfRoundsPerLoop=3
   weaponName=""
  }
 }
]
widgets {
 info {rootWidgetClass=&CanvasPanel, widgetCount=47}
 widgets [
  CanvasPanel_28 {widgetClassPath=&CanvasPanel, widget=CanvasPanel_28, properties {visibility=Visible}}
  Border_1 {
   widgetClassPath=&Border
   parent=CanvasPanel_28
   widget=Border_1
   properties {
    background {
     bIsDynamicallyLoaded=false
     drawAs=Image
     imageType=NoImage
     mirroring=NoMirror
     resourceName=None
     resourceObject=null
     tiling=NoTile
     imageSize=<32 32>
     margin=<ltrb 0 0 0 0>
     outlineSettings {
      bUseBrushTransparency=false
      roundingType=HalfHeightRadius
      width=0
      color {colorUseRule=UseColor_Specified, specifiedColor=<rgba 0 0 0 0>}
      cornerRadii=<0 0 0 0>
     }
     tintColor {colorUseRule=UseColor_Specified, specifiedColor=<rgba 0.0208 0.0208 0.0208 1>}
     uVRegion {bIsValid=false, max=<0 0>, min=<0 0>}
    }
    padding=<ltrb 0 0 0 0>
   }
   slot {
    class=CanvasPanelSlot
    properties {bAutoSize=true, layoutData {alignment=<0 0>, anchors {maximum=<0 0>, minimum=<0 0>}, offsets=<ltrb 0 0 878.7502 517.037>}}
   }
  }
  Border_0 {
   widgetClassPath=&Border
   parent=Border_1
   widget=Border_0
   properties {background {
     bIsDynamicallyLoaded=false
     drawAs=Border
     imageType=NoImage
     mirroring=NoMirror
     resourceName=None
     resourceObject=null
     tiling=NoTile
     imageSize=<32 32>
     margin=<ltrb 1 1 1 1>
     outlineSettings {
      bUseBrushTransparency=false
      roundingType=HalfHeightRadius
      width=0
      color {colorUseRule=UseColor_Specified, specifiedColor=<rgba 0 0 0 0>}
      cornerRadii=<0 0 0 0>
     }
     tintColor {colorUseRule=UseColor_Specified, specifiedColor=<rgba 0 0 0 1>}
     uVRegion {bIsValid=false, max=<0 0>, min=<0 0>}
    }}
   slot {class=BorderSlot, properties {padding=<ltrb 0 0 0 0>}}
  }
  UIKitGridPanel_82 {widgetClassPath=/Script/FortniteUI.UIKitGridPanel, parent=Border_0, widget=UIKitGridPanel_82, slot {class=BorderSlot}}
  UIKitGridPanel_1 {
   widgetClassPath=&UIKitGridPanel
   parent=UIKitGridPanel_82
   widget=UIKitGridPanel_1
   slot {class=GridSlot, properties {row=1}}
  }
  EditorUtilitySpinBox_InitialDelay {
   widgetClassPath=/Script/Blutility.EditorUtilitySpinBox
   bIsVariable=true
   parent=UIKitGridPanel_1
   widget=EditorUtilitySpinBox_InitialDelay
   properties {
    bOverride_MinValue=true
    maxFractionalDigits=1
    maxValue=20
    minFractionalDigits=0
    foregroundColor {colorUseRule=UseColor_Specified, specifiedColor=<rgba 1 1 1 1>}
   }
   slot {class=GridSlot, properties {column=1}}
  }
  TextBlock_3 {
   widgetClassPath=&TextBlock
   parent=UIKitGridPanel_1
   widget=TextBlock_3
   properties {
    text=Initial Playback Delay
    font {
     bForceMonospaced=false
     bMaterialIsStencil=false
     fontMaterial=null
     fontObject=/LauncherSocial/BrutalType
     letterSpacing=0
     monospacedWidth=1
     size=24
     skewAmount=0
     typefaceFontName=Regular
     outlineSettings {
      bApplyOutlineToDropShadows=false
      bMiteredCorners=false
      bSeparateFillAlpha=false
      outlineMaterial=null
      outlineSize=0
      outlineColor=<rgba 0 0 0 1>
     }
    }
   }
   slot {class=GridSlot}
  }
  TextBlock_0 {
   widgetClassPath=&TextBlock
   parent=UIKitGridPanel_1
   widget=TextBlock_0
   properties {
    text=Fire Rate
    font {
     bForceMonospaced=false
     bMaterialIsStencil=false
     fontMaterial=null
     fontObject=&BrutalType
     letterSpacing=0
     monospacedWidth=1
     size=24
     skewAmount=0
     typefaceFontName=Regular
     outlineSettings {
      bApplyOutlineToDropShadows=false
      bMiteredCorners=false
      bSeparateFillAlpha=false
      outlineMaterial=null
      outlineSize=0
      outlineColor=<rgba 0 0 0 1>
     }
    }
   }
   slot {class=GridSlot, properties {row=2}}
  }
  EditorUtilitySpinBox_FireRate {
   widgetClassPath=&EditorUtilitySpinBox
   bIsVariable=true
   parent=UIKitGridPanel_1
   widget=EditorUtilitySpinBox_FireRate
   properties {bOverride_MinValue=true, minValue=0.01, value=0.1, foregroundColor {colorUseRule=UseColor_Specified, specifiedColor=<rgba 1 1 1 1>}}
   slot {class=GridSlot, properties {column=1, row=2}}
  }
  EditorUtilitySpinBox_LoopDelay {
   widgetClassPath=&EditorUtilitySpinBox
   bIsVariable=true
   parent=UIKitGridPanel_1
   widget=EditorUtilitySpinBox_LoopDelay
   properties {bOverride_MinValue=true, minValue=0.0001, value=1, foregroundColor {colorUseRule=UseColor_Specified, specifiedColor=<rgba 1 1 1 1>}}
   slot {class=GridSlot, properties {column=1, row=4}}
  }
  TextBlock_4 {
   widgetClassPath=&TextBlock
   parent=UIKitGridPanel_1
   widget=TextBlock_4
   properties {
    text=Reload Time
    font {
     bForceMonospaced=false
     bMaterialIsStencil=false
     fontMaterial=null
     fontObject=&BrutalType
     letterSpacing=0
     monospacedWidth=1
     size=24
     skewAmount=0
     typefaceFontName=Regular
     outlineSettings {
      bApplyOutlineToDropShadows=false
      bMiteredCorners=false
      bSeparateFillAlpha=false
      outlineMaterial=null
      outlineSize=0
      outlineColor=<rgba 0 0 0 1>
     }
    }
   }
   slot {class=GridSlot, properties {row=4}}
  }
  EditorUtilityCheckBox_bSimulateAmmo {
   widgetClassPath=&EditorUtilityCheckBox
   bIsVariable=true
   parent=UIKitGridPanel_1
   widget=EditorUtilityCheckBox_bSimulateAmmo
   properties {checkedState=Checked}
   slot {class=GridSlot, properties {column=1, horizontalAlignment=HAlign_Center, row=5}}
  }
  HorizontalBox_2 {
   widgetClassPath=&HorizontalBox
   parent=UIKitGridPanel_1
   widget=HorizontalBox_2
   slot {class=GridSlot, properties {horizontalAlignment=HAlign_Left, row=5, verticalAlignment=VAlign_Top}}
  }
  TextBlock_6 {
   widgetClassPath=&TextBlock
   parent=HorizontalBox_2
   widget=TextBlock_6
   properties {
    text=Simulate Ammo?
    font {
     bForceMonospaced=false
     bMaterialIsStencil=false
     fontMaterial=null
     fontObject=&BrutalType
     letterSpacing=0
     monospacedWidth=1
     size=24
     skewAmount=0
     typefaceFontName=Regular
     outlineSettings {
      bApplyOutlineToDropShadows=false
      bMiteredCorners=false
      bSeparateFillAlpha=false
      outlineMaterial=null
      outlineSize=0
      outlineColor=<rgba 0 0 0 1>
     }
    }
   }
   slot {class=HorizontalBoxSlot}
  }
  UIKitGridPanel_2 {
   widgetClassPath=&UIKitGridPanel
   parent=UIKitGridPanel_82
   widget=UIKitGridPanel_2
   slot {class=GridSlot, properties {column=1, horizontalAlignment=HAlign_Center, verticalAlignment=VAlign_Center}}
  }
  VerticalBox_0 {
   widgetClassPath=&VerticalBox
   parent=UIKitGridPanel_2
   widget=VerticalBox_0
   slot {class=GridSlot, properties {columnSpan=2, row=2}}
  }
  TextBlock_1 {
   widgetClassPath=&TextBlock
   parent=VerticalBox_0
   widget=TextBlock_1
   properties {
    text=Ammo Count
    font {
     bForceMonospaced=false
     bMaterialIsStencil=false
     fontMaterial=null
     fontObject=&BrutalType
     letterSpacing=0
     monospacedWidth=1
     size=12.75
     skewAmount=0
     typefaceFontName=Regular
     outlineSettings {
      bApplyOutlineToDropShadows=false
      bMiteredCorners=false
      bSeparateFillAlpha=false
      outlineMaterial=null
      outlineSize=0
      outlineColor=<rgba 0 0 0 1>
     }
    }
   }
   slot {class=VerticalBoxSlot, properties {horizontalAlignment=HAlign_Center}}
  }
  EditorUtilitySpinBox_AmmoCount {
   widgetClassPath=&EditorUtilitySpinBox
   bIsVariable=true
   parent=VerticalBox_0
   widget=EditorUtilitySpinBox_AmmoCount
   properties {
    bOverride_MaxValue=true
    bOverride_MinValue=true
    maxFractionalDigits=0
    maxValue=30
    minFractionalDigits=0
    value=30
    foregroundColor {colorUseRule=UseColor_Specified, specifiedColor=<rgba 1 1 1 1>}
    widgetStyle {
     activeBackgroundBrush {
      bIsDynamicallyLoaded=false
      drawAs=RoundedBox
      imageType=NoImage
      mirroring=NoMirror
      resourceName=None
      resourceObject=null
      tiling=NoTile
      imageSize=<0 0>
      margin=<ltrb 0 0 0 0>
      outlineSettings {
       bUseBrushTransparency=false
       roundingType=FixedRadius
       width=1
       color {colorUseRule=UseColor_ColorTable, specifiedColor=<rgba 0 0 0 0>}
       cornerRadii=<4 4 4 4>
      }
      tintColor {colorUseRule=UseColor_Specified, specifiedColor=<rgba 0.0048 0.0048 0.0048 1>}
      uVRegion {bIsValid=false, max=<0 0>, min=<0 0>}
     }
     activeFillBrush {
      bIsDynamicallyLoaded=false
      drawAs=RoundedBox
      imageType=NoImage
      mirroring=NoMirror
      resourceName=None
      resourceObject=null
      tiling=NoTile
      imageSize=<0 0>
      margin=<ltrb 0 0 0 0>
      outlineSettings {
       bUseBrushTransparency=false
       roundingType=FixedRadius
       width=1
       color {colorUseRule=UseColor_Specified, specifiedColor=<rgba 0 0 0 0>}
       cornerRadii=<4 4 4 4>
      }
      tintColor {colorUseRule=UseColor_Specified, specifiedColor=<rgba 0 0.162 0.7454 1>}
      uVRegion {bIsValid=false, max=<0 0>, min=<0 0>}
     }
     arrowsImage {
      bIsDynamicallyLoaded=false
      drawAs=NoDrawType
      imageType=NoImage
      mirroring=NoMirror
      resourceName=None
      resourceObject=null
      tiling=NoTile
      imageSize=<0 0>
      margin=<ltrb 0 0 0 0>
      outlineSettings {
       bUseBrushTransparency=false
       roundingType=HalfHeightRadius
       width=0
       color {colorUseRule=UseColor_Specified, specifiedColor=<rgba 0 0 0 0>}
       cornerRadii=<0 0 0 0>
      }
      tintColor {colorUseRule=UseColor_Specified, specifiedColor=<rgba 1 1 1 1>}
      uVRegion {bIsValid=false, max=<0 0>, min=<0 0>}
     }
     backgroundBrush {
      bIsDynamicallyLoaded=false
      drawAs=RoundedBox
      imageType=NoImage
      mirroring=NoMirror
      resourceName=None
      resourceObject=null
      tiling=NoTile
      imageSize=<0 0>
      margin=<ltrb 0 0 0 0>
      outlineSettings {
       bUseBrushTransparency=false
       roundingType=FixedRadius
       width=1
       color {colorUseRule=UseColor_ColorTable, specifiedColor=<rgba 0 0 0 0>}
       cornerRadii=<4 4 4 4>
      }
      tintColor {colorUseRule=UseColor_Specified, specifiedColor=<rgba 0.0048 0.0048 0.0048 1>}
      uVRegion {bIsValid=false, max=<0 0>, min=<0 0>}
     }
     foregroundColor {colorUseRule=UseColor_Specified, specifiedColor=<rgba 1 1 1 1>}
     hoveredBackgroundBrush {
      bIsDynamicallyLoaded=false
      drawAs=RoundedBox
      imageType=NoImage
      mirroring=NoMirror
      resourceName=None
      resourceObject=null
      tiling=NoTile
      imageSize=<0 0>
      margin=<ltrb 0 0 0 0>
      outlineSettings {
       bUseBrushTransparency=false
       roundingType=FixedRadius
       width=1
       color {colorUseRule=UseColor_ColorTable, specifiedColor=<rgba 0 0 0 0>}
       cornerRadii=<4 4 4 4>
      }
      tintColor {colorUseRule=UseColor_Specified, specifiedColor=<rgba 0.0048 0.0048 0.0048 1>}
      uVRegion {bIsValid=false, max=<0 0>, min=<0 0>}
     }
     hoveredFillBrush {
      bIsDynamicallyLoaded=false
      drawAs=RoundedBox
      imageType=NoImage
      mirroring=NoMirror
      resourceName=None
      resourceObject=null
      tiling=NoTile
      imageSize=<0 0>
      margin=<ltrb 0 0 0 0>
      outlineSettings {
       bUseBrushTransparency=false
       roundingType=FixedRadius
       width=1
       color {colorUseRule=UseColor_Specified, specifiedColor=<rgba 0 0 0 0>}
       cornerRadii=<4 4 4 4>
      }
      tintColor {colorUseRule=UseColor_Specified, specifiedColor=<rgba 0.0953 0.0953 0.0953 1>}
      uVRegion {bIsValid=false, max=<0 0>, min=<0 0>}
     }
     inactiveFillBrush {
      bIsDynamicallyLoaded=false
      drawAs=RoundedBox
      imageType=NoImage
      mirroring=NoMirror
      resourceName=None
      resourceObject=null
      tiling=NoTile
      imageSize=<0 0>
      margin=<ltrb 0 0 0 0>
      outlineSettings {
       bUseBrushTransparency=false
       roundingType=FixedRadius
       width=1
       color {colorUseRule=UseColor_Specified, specifiedColor=<rgba 0 0 0 0>}
       cornerRadii=<4 4 4 4>
      }
      tintColor {colorUseRule=UseColor_Specified, specifiedColor=<rgba 0.0395 0.0395 0.0395 1>}
      uVRegion {bIsValid=false, max=<0 0>, min=<0 0>}
     }
     insetPadding=<ltrb 1 1 1 1>
     textPadding=<ltrb 7 2.5 7 2.5>
    }
   }
   slot {class=VerticalBoxSlot}
  }
  AudioButtonToggle_DefaultMix {
   widgetClassPath=/AudioWidgets/AudioButtonToggle/AudioButtonToggle
   bIsVariable=true
   parent=UIKitGridPanel_2
   widget=AudioButtonToggle_DefaultMix
   properties {textValue="Default Mix (wip)", visibility=Hidden}
   slot {class=GridSlot, properties {row=1}}
  }
  AudioButtonToggle_AimDownSights {
   widgetClassPath=&AudioButtonToggle
   bIsVariable=true
   parent=UIKitGridPanel_2
   widget=AudioButtonToggle_AimDownSights
   properties {textValue=AimDownSights}
   slot {class=GridSlot, properties {row=1}}
  }
  TextBlock_7 {
   widgetClassPath=&TextBlock
   parent=UIKitGridPanel_2
   widget=TextBlock_7
   properties {
    text=Controls
    font {
     bForceMonospaced=false
     bMaterialIsStencil=false
     fontMaterial=null
     fontObject=&BrutalType
     letterSpacing=0
     monospacedWidth=1
     size=24
     skewAmount=0
     typefaceFontName=Regular
     outlineSettings {
      bApplyOutlineToDropShadows=false
      bMiteredCorners=false
      bSeparateFillAlpha=false
      outlineMaterial=null
      outlineSize=0
      outlineColor=<rgba 0 0 0 1>
     }
    }
   }
   slot {class=GridSlot, properties {columnSpan=2, horizontalAlignment=HAlign_Center}}
  }
  HorizontalBox_34 {
   widgetClassPath=&HorizontalBox
   parent=UIKitGridPanel_2
   widget=HorizontalBox_34
   slot {class=GridSlot, properties {column=1, horizontalAlignment=HAlign_Center, row=1}}
  }
  DistanceParam {
   widgetClassPath=/AudioWidgets/AudioKnobSmall/AudioKnobSmall
   bIsVariable=true
   parent=HorizontalBox_34
   widget=DistanceParam
   properties {controlValueMax=12000, displayMax=12000, label=Distance, maxFractionalDigits=1, maxIntegralDigits=5, units=""}
   slot {class=HorizontalBoxSlot, properties {horizontalAlignment=HAlign_Right}}
  }
  AzimuthParam {
   widgetClassPath=&AudioKnobSmall
   bIsVariable=true
   parent=HorizontalBox_34
   widget=AzimuthParam
   properties {
    controlValueMax=180
    controlValueMin=-180
    displayMax=180
    displayMin=-180
    label=Azimuth
    maxFractionalDigits=0
    units=""
    foregroundColor {colorUseRule=UseColor_Specified, specifiedColor=<rgba 0.2818 1 0.75 1>}
   }
   slot {class=HorizontalBoxSlot}
  }
  EUW_MetaSoundFloat_Slider_MasterVolume {
   widgetClassPath=/TechAudioToolsContent/Widgets/MetaSoundLiterals/Float/EUW_MetaSoundFloat_Slider_DefaultVertical
   bIsVariable=true
   parent=HorizontalBox_34
   widget=EUW_MetaSoundFloat_Slider_MasterVolume
   properties {
    apply Logarithmic Interpolation=true
    input Name=MasterVolume
    label Text=Master Volume
    metaSoundLiteralViewModel_Float=&self:WidgetTree.EUW_MetaSoundFloat_Slider_MasterVolume.MetaSoundLiteralViewModel_Float_0
    width Override=96
    slider Style {
     material=null
     desiredSize=<25 250>
     sliderBackgroundAccentColor=<rgba 0.005 0.005 0.005 1>
     sliderBackgroundColor=<rgba 0.008 0.008 0.008 1>
     sliderHandleMainColor=<rgba 1 1 1 1>
     sliderHandleOutlineColor=<rgba 0.15 0.15 0.15 1>
     sliderValueMainColor=<rgba 0.3278 0 1 1>
     textBoxStyle {
      backgroundColor {colorUseRule=UseColor_ColorTable, specifiedColor=<rgba 0 0 0 0>}
      backgroundImage {
       bIsDynamicallyLoaded=false
       drawAs=RoundedBox
       imageType=NoImage
       mirroring=NoMirror
       resourceName=None
       resourceObject=null
       tiling=NoTile
       imageSize=<56 28>
       margin=<ltrb 0 0 0 0>
       outlineSettings {
        bUseBrushTransparency=false
        roundingType=FixedRadius
        width=0
        color {colorUseRule=UseColor_Specified, specifiedColor=<rgba 0 0 0 0>}
        cornerRadii=<4 4 4 4>
       }
       tintColor {colorUseRule=UseColor_ColorTable, specifiedColor=<rgba 0 0 0 0>}
       uVRegion {bIsValid=false, max=<0 0>, min=<0 0>}
      }
     }
    }
   }
   slot {class=HorizontalBoxSlot}
  }
  EUW_MetaSoundFloat_Slider_MechVolume {
   widgetClassPath=&EUW_MetaSoundFloat_Slider_DefaultVertical
   bIsVariable=true
   parent=HorizontalBox_34
   widget=EUW_MetaSoundFloat_Slider_MechVolume
   properties {
    apply Logarithmic Interpolation=true
    input Name=Mech.Volume
    label Text=Mech
    metaSoundLiteralViewModel_Float=&self:WidgetTree.EUW_MetaSoundFloat_Slider_MechVolume.MetaSoundLiteralViewModel_Float_0
    slider Style {
     material=null
     desiredSize=<25 250>
     sliderBackgroundAccentColor=<rgba 0.005 0.005 0.005 1>
     sliderBackgroundColor=<rgba 0.008 0.008 0.008 1>
     sliderHandleMainColor=<rgba 1 1 1 1>
     sliderHandleOutlineColor=<rgba 0.15 0.15 0.15 1>
     sliderValueMainColor=<rgba 0.3278 0 1 1>
     textBoxStyle {
      backgroundColor {colorUseRule=UseColor_ColorTable, specifiedColor=<rgba 0 0 0 0>}
      backgroundImage {
       bIsDynamicallyLoaded=false
       drawAs=RoundedBox
       imageType=NoImage
       mirroring=NoMirror
       resourceName=None
       resourceObject=null
       tiling=NoTile
       imageSize=<56 28>
       margin=<ltrb 0 0 0 0>
       outlineSettings {
        bUseBrushTransparency=false
        roundingType=FixedRadius
        width=0
        color {colorUseRule=UseColor_Specified, specifiedColor=<rgba 0 0 0 0>}
        cornerRadii=<4 4 4 4>
       }
       tintColor {colorUseRule=UseColor_ColorTable, specifiedColor=<rgba 0 0 0 0>}
       uVRegion {bIsValid=false, max=<0 0>, min=<0 0>}
      }
     }
    }
   }
   slot {class=HorizontalBoxSlot}
  }
  EUW_MetaSoundFloat_Slider_LowAmmoVolume {
   widgetClassPath=&EUW_MetaSoundFloat_Slider_DefaultVertical
   bIsVariable=true
   parent=HorizontalBox_34
   widget=EUW_MetaSoundFloat_Slider_LowAmmoVolume
   properties {
    apply Logarithmic Interpolation=true
    input Name=LowAmmo.Volume
    label Text=Low Ammo
    metaSoundLiteralViewModel_Float=&self:WidgetTree.EUW_MetaSoundFloat_Slider_LowAmmoVolume.MetaSoundLiteralViewModel_Float_0
    slider Style {
     material=null
     desiredSize=<25 250>
     sliderBackgroundAccentColor=<rgba 0.005 0.005 0.005 1>
     sliderBackgroundColor=<rgba 0.008 0.008 0.008 1>
     sliderHandleMainColor=<rgba 1 1 1 1>
     sliderHandleOutlineColor=<rgba 0.15 0.15 0.15 1>
     sliderValueMainColor=<rgba 0.3278 0 1 1>
     textBoxStyle {
      backgroundColor {colorUseRule=UseColor_ColorTable, specifiedColor=<rgba 0 0 0 0>}
      backgroundImage {
       bIsDynamicallyLoaded=false
       drawAs=RoundedBox
       imageType=NoImage
       mirroring=NoMirror
       resourceName=None
       resourceObject=null
       tiling=NoTile
       imageSize=<56 28>
       margin=<ltrb 0 0 0 0>
       outlineSettings {
        bUseBrushTransparency=false
        roundingType=FixedRadius
        width=0
        color {colorUseRule=UseColor_Specified, specifiedColor=<rgba 0 0 0 0>}
        cornerRadii=<4 4 4 4>
       }
       tintColor {colorUseRule=UseColor_ColorTable, specifiedColor=<rgba 0 0 0 0>}
       uVRegion {bIsValid=false, max=<0 0>, min=<0 0>}
      }
     }
    }
   }
   slot {class=HorizontalBoxSlot}
  }
  UniformGridPanel_0 {widgetClassPath=/Script/UMG.UniformGridPanel, parent=UIKitGridPanel_82, widget=UniformGridPanel_0, slot {class=GridSlot, properties {row=2}}}
  HorizontalBox_0 {widgetClassPath=&HorizontalBox, parent=UIKitGridPanel_82, widget=HorizontalBox_0, slot {class=GridSlot, properties {column=2}}}
  UIKitGridPanel_0 {
   widgetClassPath=&UIKitGridPanel
   parent=UIKitGridPanel_82
   widget=UIKitGridPanel_0
   slot {class=GridSlot, properties {column=1, row=1, verticalAlignment=VAlign_Top}}
  }
  EditorUtilityButton_Update {
   widgetClassPath=&EditorUtilityButton
   parent=UIKitGridPanel_0
   widget=EditorUtilityButton_Update
   properties {toolTipText=Updates the playback data and the selected metasound to the slot.}
   slot {class=GridSlot, properties {row=2}}
  }
  TextBlock_8 {
   widgetClassPath=&TextBlock
   parent=EditorUtilityButton_Update
   widget=TextBlock_8
   properties {
    justification=Center
    text=Update
    textTransformPolicy=ToUpper
    font {
     bForceMonospaced=false
     bMaterialIsStencil=false
     fontMaterial=null
     fontObject=&BrutalType
     letterSpacing=0
     monospacedWidth=1
     size=24
     skewAmount=0
     typefaceFontName=Regular
     outlineSettings {
      bApplyOutlineToDropShadows=false
      bMiteredCorners=false
      bSeparateFillAlpha=false
      outlineMaterial=null
      outlineSize=0
      outlineColor=<rgba 0 0 0 1>
     }
    }
   }
   slot {class=ButtonSlot, properties {horizontalAlignment=HAlign_Fill, verticalAlignment=VAlign_Fill}}
  }
  EditorUtilityButton_Delete {
   widgetClassPath=&EditorUtilityButton
   parent=UIKitGridPanel_0
   widget=EditorUtilityButton_Delete
   slot {class=GridSlot, properties {row=3}}
  }
  TextBlock_Play_1 {
   widgetClassPath=&TextBlock
   bIsVariable=true
   parent=EditorUtilityButton_Delete
   widget=TextBlock_Play_1
   properties {
    justification=Center
    text=Delete
    textTransformPolicy=ToUpper
    font {
     bForceMonospaced=false
     bMaterialIsStencil=false
     fontMaterial=null
     fontObject=&BrutalType
     letterSpacing=0
     monospacedWidth=1
     size=24
     skewAmount=0
     typefaceFontName=Regular
     outlineSettings {
      bApplyOutlineToDropShadows=false
      bMiteredCorners=false
      bSeparateFillAlpha=false
      outlineMaterial=null
      outlineSize=0
      outlineColor=<rgba 0 0 0 1>
     }
    }
   }
   slot {class=ButtonSlot, properties {horizontalAlignment=HAlign_Fill, verticalAlignment=VAlign_Fill}}
  }
  SinglePropertyView_WeaponItemDefinition {
   widgetClassPath=&SinglePropertyView
   parent=UIKitGridPanel_0
   widget=SinglePropertyView_WeaponItemDefinition
   properties {propertyName=FortRangedWeaponItemDefinition}
   slot {class=GridSlot, properties {horizontalAlignment=HAlign_Center, row=4, verticalAlignment=VAlign_Center}}
  }
  EditorUtilityButton_PlayButton {
   widgetClassPath=&EditorUtilityButton
   parent=UIKitGridPanel_0
   widget=EditorUtilityButton_PlayButton
   slot {class=GridSlot, properties {verticalAlignment=VAlign_Top}}
  }
  TextBlock_Play {
   widgetClassPath=&TextBlock
   bIsVariable=true
   parent=EditorUtilityButton_PlayButton
   widget=TextBlock_Play
   properties {
    justification=Center
    text=Play
    textTransformPolicy=ToUpper
    font {
     bForceMonospaced=false
     bMaterialIsStencil=false
     fontMaterial=null
     fontObject=&BrutalType
     letterSpacing=0
     monospacedWidth=1
     size=24
     skewAmount=0
     typefaceFontName=Regular
     outlineSettings {
      bApplyOutlineToDropShadows=false
      bMiteredCorners=false
      bSeparateFillAlpha=false
      outlineMaterial=null
      outlineSize=0
      outlineColor=<rgba 0 0 0 1>
     }
    }
   }
   slot {class=ButtonSlot, properties {horizontalAlignment=HAlign_Fill, verticalAlignment=VAlign_Fill}}
  }
  EditorUtilityButton_StopButton {
   widgetClassPath=&EditorUtilityButton
   parent=UIKitGridPanel_0
   widget=EditorUtilityButton_StopButton
   slot {class=GridSlot, properties {row=1}}
  }
  TextBlock_5 {
   widgetClassPath=&TextBlock
   parent=EditorUtilityButton_StopButton
   widget=TextBlock_5
   properties {
    justification=Center
    text=Stop
    textTransformPolicy=ToUpper
    font {
     bForceMonospaced=false
     bMaterialIsStencil=false
     fontMaterial=null
     fontObject=&BrutalType
     letterSpacing=0
     monospacedWidth=1
     size=24
     skewAmount=0
     typefaceFontName=Regular
     outlineSettings {
      bApplyOutlineToDropShadows=false
      bMiteredCorners=false
      bSeparateFillAlpha=false
      outlineMaterial=null
      outlineSize=0
      outlineColor=<rgba 0 0 0 1>
     }
    }
   }
   slot {class=ButtonSlot, properties {horizontalAlignment=HAlign_Fill, verticalAlignment=VAlign_Fill}}
  }
  UIKitGridPanel_3 {
   widgetClassPath=&UIKitGridPanel
   parent=UIKitGridPanel_82
   widget=UIKitGridPanel_3
   slot {class=GridSlot, properties {horizontalAlignment=HAlign_Center, verticalAlignment=VAlign_Center}}
  }
  AudioButtonToggle_PerspectiveToggle {
   widgetClassPath=&AudioButtonToggle
   bIsVariable=true
   parent=UIKitGridPanel_3
   widget=AudioButtonToggle_PerspectiveToggle
   properties {textValue=Perspective Switch}
   slot {class=GridSlot, properties {columnSpan=3, row=2}}
  }
  Border_3P {
   widgetClassPath=&Border
   bIsVariable=true
   parent=UIKitGridPanel_3
   widget=Border_3P
   properties {background {
     bIsDynamicallyLoaded=false
     drawAs=Image
     imageType=NoImage
     mirroring=NoMirror
     resourceName=None
     resourceObject=null
     tiling=NoTile
     imageSize=<32 32>
     margin=<ltrb 0 0 0 0>
     outlineSettings {
      bUseBrushTransparency=false
      roundingType=HalfHeightRadius
      width=0
      color {colorUseRule=UseColor_Specified, specifiedColor=<rgba 0 0 0 0>}
      cornerRadii=<0 0 0 0>
     }
     tintColor {colorUseRule=UseColor_Specified, specifiedColor=<rgba 1 1 1 0>}
     uVRegion {bIsValid=false, max=<0 0>, min=<0 0>}
    }}
   slot {class=GridSlot}
  }
  Overlay_62 {widgetClassPath=&Overlay, parent=Border_3P, widget=Overlay_62, slot {class=BorderSlot}}
  SinglePropertyView_WeaponFireSound_3P {
   widgetClassPath=&SinglePropertyView
   parent=Overlay_62
   widget=SinglePropertyView_WeaponFireSound_3P
   properties {propertyName=3PWeaponSound}
   slot {class=OverlaySlot, properties {horizontalAlignment=HAlign_Center, verticalAlignment=VAlign_Fill}}
  }
  Border_1P {
   widgetClassPath=&Border
   bIsVariable=true
   parent=UIKitGridPanel_3
   widget=Border_1P
   properties {
    horizontalAlignment=HAlign_Center
    background {
     bIsDynamicallyLoaded=false
     drawAs=Image
     imageType=NoImage
     mirroring=NoMirror
     resourceName=None
     resourceObject=null
     tiling=NoTile
     imageSize=<32 32>
     margin=<ltrb 0 0 0 0>
     outlineSettings {
      bUseBrushTransparency=false
      roundingType=HalfHeightRadius
      width=0
      color {colorUseRule=UseColor_Specified, specifiedColor=<rgba 0 0 0 0>}
      cornerRadii=<0 0 0 0>
     }
     tintColor {colorUseRule=UseColor_Specified, specifiedColor=<rgba 1 1 1 0>}
     uVRegion {bIsValid=false, max=<0 0>, min=<0 0>}
    }
    padding=<ltrb 0 0 0 0>
   }
   slot {class=GridSlot, properties {row=1}}
  }
  Overlay_64 {
   widgetClassPath=&Overlay
   parent=Border_1P
   widget=Overlay_64
   slot {class=BorderSlot, properties {horizontalAlignment=HAlign_Center, padding=<ltrb 0 0 0 0>}}
  }
  SinglePropertyView_WeaponFireSound_1P {
   widgetClassPath=&SinglePropertyView
   parent=Overlay_64
   widget=SinglePropertyView_WeaponFireSound_1P
   properties {propertyName=1PWeaponSound}
   slot {class=OverlaySlot}
  }
 ]
}
