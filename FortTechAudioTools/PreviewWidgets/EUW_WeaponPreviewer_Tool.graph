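# Editor utility widget for auditioning Fortnite weapon fire sounds, hosting EUW_WeaponPreviewer_Module widgets in a scroll box with add, clear, play-all and stop-all controls.
# A combo box built from the hardcoded Paprika weapon array adds the chosen weapon in first person.
# Its "Firefight" option instead adds every weapon in third person with randomized initial delays to stagger playback.
# A hidden, work-in-progress Default Mix toggle calls control bus mix activation with no mix assigned.
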EditorUtilityWidgetBlueprint /FortTechAudioTools/PreviewWidgets/EUW_WeaponPreviewer_Tool
graphs [
 CreateWeaponModuleWidget {nodes [
   n1 = CreateWeaponModuleWidget!() -> (then, WeaponPlaybackStruct:Struct Weapon Playback Data Structure, 1P?:bool=true) {is=title K2Node_FunctionEntry}
   > n2 = Construct Object from Class(Class=&/EUW_WeaponPreviewer_Module.EUW_WeaponPreviewer_Module_C) -> (ReturnValue:EUW Weapon Previewer Module Object Reference)
   > Set WidgetPlaybackData!(execute, WidgetPlaybackData:Struct Weapon Playback Data Structure=@n1, self:EUW Weapon Previewer Module Object Reference=@n2) -> (then, Output_Get:Struct Weapon Playback Data Structure)
   > AddChild(self=Get MainBox() -> (MainBox:Editor Utility Scroll Box Object Reference), Content=@n2)
   > Bind Event to On Request Close(self:EUW Weapon Previewer Module Object Reference=@n2, Delegate=Create Event()) {is=title K2Node_AddDelegate}
   > n3 = Branch(Condition=@n1.1P?)
   n4 = Get AudioButtonToggle_PerspectiveToggle!(self:EUW Weapon Previewer Module Object Reference=@n2) -> (AudioButtonToggle_PerspectiveToggle:Audio Button Toggle Object Reference)
   > SetDeselected(self:Audio Button Toggle Object Reference=@n4)
   SetSelected!(execute=@n3.else, self:Audio Button Toggle Object Reference=@n4, ReportEvent:bool=true) -> (then)
  ]}
 EventGraph {nodes [
   n1 = Get AudioButtonToggle_DefaultMix() -> (AudioButtonToggle_DefaultMix:Audio Button Toggle Object Reference)
   n2 = SetSelected(self=@n1, ReportEvent=true)
   Activate Control Bus Mix!(execute=[@n2.then, @n5.then], WorldContextObject:Object Reference, Mix:Sound Control Bus Mix Object Reference) -> (then)
   "On Clicked (AddButton)"() {is=title K2Node_ComponentBoundEvent}
   > CreateWeaponModuleWidget!(execute, self, WeaponPlaybackStruct:Struct Weapon Playback Data Structure, 1P?:bool=true) -> (then) {is=title K2Node_CallFunction}
   "On Clicked (ClearButton)"() {is=title K2Node_ComponentBoundEvent}
   > ClearChildren(self=Get MainBox())
   "On Clicked (AddButton_AddPaprikaWeapons)"() {is=title K2Node_ComponentBoundEvent}
   > CreatePaprikaWeaponInstances()
   n3 = "On Selected (AudioButtonToggle_DefaultMix)"!() -> (OutputDelegate, then, Button:Audio Button Toggle Object Reference) {is=title K2Node_ComponentBoundEvent}
   n4 = GetIsSelected!(execute=[@n3.then, @n6.then], self:Audio Button Toggle Object Reference=@n1) -> (then, IsSelected:bool)
   > n5 = Branch(Condition=@n4)
   Deactivate Control Bus Mix!(execute=@n5.else, WorldContextObject:Object Reference, Mix:Sound Control Bus Mix Object Reference) -> (then)
   n6 = "On Deselected (AudioButtonToggle_DefaultMix)"!() -> (OutputDelegate, then, Button:Audio Button Toggle Object Reference) {is=title K2Node_ComponentBoundEvent}
   "On Pressed (AddButton_PlayAll)"() {is=title K2Node_ComponentBoundEvent}
   > PlayAllWidgets()
   "On Pressed (AddButton_StopAll)"() {is=title K2Node_ComponentBoundEvent}
   > StopAllWidgets()
   CreatePaprikaWeaponInstances() {is=title K2Node_CustomEvent}
   n7 = Get EditorUtilityComboBoxString_0() -> (EditorUtilityComboBoxString_0:Editor Utility Combo Box String Object Reference)
   > n8 = Branch(Condition="Not Equal Exactly (String)"(A=GetSelectedOption!(self:"Combo Box (String) Object Reference"=@n7) -> (ReturnValue:String), B=Firefight))
   > n9 = For Each Loop(Array:Array of Struct Weapon Playback Data Structures=Get Weapons Structs() -> (Weapons Structs:Array of Struct Weapon Playback Data Structures)) -> (Array Element:Struct Weapon Playback Data Structure)
   n10 = Get EditorUtilityComboBoxString_0()
   > Branch(Condition="Equal Exactly (String)"(
     A=Break Struct Weapon Playback Data!(Struct_WeaponPlaybackData:"Struct Weapon Playback Data Structure (by ref)"=@n9) -> (
      WeaponName_17_#1:String
      3P_WeaponSound_33_#2:Sound Base Object Reference
      1P_WeaponSound_30_#3:Sound Base Object Reference
      IntialDelay_20_#4:double=~
      NumberOfRoundsPerLoop_4_#5:int
      FireRate_5_#6:double=~
      LoopCount_8_#7:int
      LoopDelay_11_#8:double=~
      AmmoCount_23_#9:int
      AssociatedItemDefinition_26_#10:Fort Weapon Ranged Item Definition Object Reference
     ) {is=title K2Node_BreakStruct}
     B=GetSelectedOption(self=@n10)
    ))
   > CreateWeaponModuleWidget(WeaponPlaybackStruct=@n9, 1P?=true)
   n11 = For Each Loop(Exec=@n8.else, Array=Get Weapons Structs())
   n12 = Break Struct Weapon Playback Data(Struct_WeaponPlaybackData=@n11) -> (
    IntialDelay_20_#4=~
    FireRate_5_#6=~
    LoopDelay_11_#8=~
   )
   > CreateWeaponModuleWidget(
    WeaponPlaybackStruct=Make Struct Weapon Playback Data!(
     WeaponName_17_#1:String=@n12
     3P_WeaponSound_33_#2:Sound Base Object Reference=@n12.3P_WeaponSound_33_#2
     1P_WeaponSound_30_#3:Sound Base Object Reference=@n12.1P_WeaponSound_30_#3
     IntialDelay_20_#4:double=RandomFloatInRange(Min=1, Max=3)
     NumberOfRoundsPerLoop_4_#5:int=@n12.NumberOfRoundsPerLoop_4_#5
     FireRate_5_#6:double=@n12.FireRate_5_#6
     LoopCount_8_#7:int=@n12.LoopCount_8_#7
     LoopDelay_11_#8:double=@n12.LoopDelay_11_#8
     AmmoCount_23_#9:int=@n12.AmmoCount_23_#9
     AssociatedItemDefinition_26_#10:Fort Weapon Ranged Item Definition Object Reference=@n12.AssociatedItemDefinition_26_#10
    ) -> (Struct_WeaponPlaybackData:Struct Weapon Playback Data Structure) {is=title K2Node_MakeStruct}
    1P?=false
   )
   n13 = ClearIndividualChildWidget!() -> (OutputDelegate, then, Module:EUW Weapon Previewer Module Object Reference) {is=title K2Node_CustomEvent}
   > RemoveChild(self=Get MainBox(), Content=@n13.Module)
   Event Construct()
   > n14 = For Each Loop(Array=Get Weapons Structs())
   n15 = Get EditorUtilityComboBoxString_0()
   SetSelectedOption!(
    execute=@n14.Completed
    self:"Combo Box (String) Object Reference"=@n15
    Option:String=Get DefaultPaprikaWeapon() -> (DefaultPaprikaWeapon:String)
   ) -> (then)
   AddOption!(
    execute=@n14.LoopBody
    self:"Combo Box (String) Object Reference"=@n15
    Option:String=Break Struct Weapon Playback Data(Struct_WeaponPlaybackData=@n14) -> (
     IntialDelay_20_#4=~
     FireRate_5_#6=~
     LoopDelay_11_#8=~
    )
   ) -> (then)
  ]}
 PlayAllWidgets {nodes [
   PlayAllWidgets() {is=title K2Node_FunctionEntry}
   n1 = GetAllChildren(self=Get MainBox())
   > Branch(Condition=Is Not Empty!(TargetArray:Array of Widget Object References=@n1) -> (ReturnValue:bool))
   > n2 = For Each Loop(Array:Array of Widget Object References=@n1) -> (Array Element:Widget Object Reference)
   > n3 = Cast To EUW_WeaponPreviewer_Module!(execute, Object=@n2) -> (then, CastFailed, AsEUW Weapon Previewer Module:EUW Weapon Previewer Module Object Reference) {is=title K2Node_DynamicCast}
   > StartPreviewingSounds(self:EUW Weapon Previewer Module Object Reference=@n3)
  ]}
 StopAllWidgets {nodes [
   StopAllWidgets() {is=title K2Node_FunctionEntry}
   n1 = GetAllChildren(self=Get MainBox())
   > Branch(Condition=Is Not Empty(TargetArray=@n1))
   > n2 = For Each Loop(Array=@n1)
   > n3 = Cast To EUW_WeaponPreviewer_Module(Object=@n2)
   > StopPreviewer(self:EUW Weapon Previewer Module Object Reference=@n3)
  ]}
]
variables [
 DefaultPaprikaWeapon {type=string, default=SCAR}
 Weapons Structs {
  type="array<Struct_WeaponPlaybackData>"
  default [
   {
    1P_WeaponSound=/Game/QA/Art_QA/BR_Audio/Audio/MSS_BR_Audio_Paprika_DMR_Fire_1p
    3P_WeaponSound=/Game/QA/Art_QA/BR_Audio/Audio/MSS_BR_Audio_Paprika_DMR_Fire_3p
    ammoCount=30
    associatedItemDefinition=/Game/Athena/Items/Weapons/WID_Sniper_Suppressed_Scope_Athena_VR_Ore_T03
    fireRate=0.1
    intialDelay=0
    loop Count=2
    loop Delay=1
    numberOfRoundsPerLoop=3
    weaponName=DMR
   }
   {
    1P_WeaponSound=/Game/QA/Art_QA/BR_Audio/Audio/MSS_BR_Audio_Paprika_BurstSMG_Fire_1p
    3P_WeaponSound=/Game/QA/Art_QA/BR_Audio/Audio/MSS_BR_Audio_Paprika_BurstSMG_Fire_3p
    ammoCount=30
    associatedItemDefinition=/Game/Athena/Items/Weapons/WID_Pistol_BurstFireSMG_Athena_C_Ore_T03
    fireRate=0.1
    intialDelay=0
    loop Count=2
    loop Delay=1
    numberOfRoundsPerLoop=3
    weaponName=Burst SMG
   }
   {
    1P_WeaponSound=/Game/QA/Art_QA/BR_Audio/Audio/MSS_BR_Audio_Paprika_DPSShotgun_Fire_1p
    3P_WeaponSound=/Game/QA/Art_QA/BR_Audio/Audio/MSS_BR_Audio_Paprika_DPSShotgun_Fire_3p
    ammoCount=30
    associatedItemDefinition=/Game/Athena/Items/Weapons/WID_Shotgun_AutoDrum_Athena_C_Ore_T03
    fireRate=0.1
    intialDelay=0
    loop Count=2
    loop Delay=1
    numberOfRoundsPerLoop=3
    weaponName=DPS Shotgun
   }
   {
    1P_WeaponSound=/Game/QA/Art_QA/BR_Audio/Audio/MSS_BR_Audio_Paprika_HeavyAR_Fire_1p
    3P_WeaponSound=/Game/QA/Art_QA/BR_Audio/Audio/MSS_BR_Audio_Paprika_HeavyAR_Fire_3p
    ammoCount=30
    associatedItemDefinition=/Game/Athena/Items/Weapons/WID_Assault_Heavy_Athena_SR_Ore_T03
    fireRate=0.1
    intialDelay=0
    loop Count=2
    loop Delay=1
    numberOfRoundsPerLoop=3
    weaponName=Heavy AR
   }
   {
    1P_WeaponSound=/Game/QA/Art_QA/BR_Audio/Audio/MSS_BR_Audio_Paprika_SCAR_Fire_1p
    3P_WeaponSound=/Game/QA/Art_QA/BR_Audio/Audio/MSS_BR_Audio_Paprika_SCAR_Fire_3p
    ammoCount=30
    associatedItemDefinition=/Game/Athena/Items/Weapons/WID_Assault_SemiAuto_Athena_R_Ore_T03
    fireRate=0.1
    intialDelay=0
    loop Count=2
    loop Delay=1
    numberOfRoundsPerLoop=3
    weaponName=SCAR
   }
   {
    1P_WeaponSound=/Game/QA/Art_QA/BR_Audio/Audio/MSS_BR_Audio_Paprika_Pistol_Fire_1p
    3P_WeaponSound=/Game/QA/Art_QA/BR_Audio/Audio/MSS_BR_Audio_Paprika_Pistol_Fire_3p
    ammoCount=30
    associatedItemDefinition=/Game/Athena/Items/Weapons/WID_DualPistol_Suppresseed_Athena_UC_T01
    fireRate=0.1
    intialDelay=0
    loop Count=2
    loop Delay=1
    numberOfRoundsPerLoop=3
    weaponName=Pistol
   }
   {
    1P_WeaponSound=/Game/QA/Art_QA/BR_Audio/Audio/MSS_BR_Audio_Paprika_InfantryAR_Fire_1p
    3P_WeaponSound=/Game/QA/Art_QA/BR_Audio/Audio/MSS_BR_Audio_Paprika_InfantryAR_Fire_3p
    ammoCount=30
    associatedItemDefinition=/Game/Athena/Items/Weapons/WID_Assault_Infantry_Athena_C
    fireRate=0.1
    intialDelay=0
    loop Count=2
    loop Delay=1
    numberOfRoundsPerLoop=3
    weaponName=Infantry AR
   }
   {
    1P_WeaponSound=/Game/QA/Art_QA/BR_Audio/Audio/MSS_BR_Audio_Paprika_PumpShotgun_Fire_1p
    3P_WeaponSound=&MSS_BR_Audio_Paprika_SCAR_Fire_3p
    ammoCount=30
    associatedItemDefinition=/Game/Athena/Items/Weapons/WID_Shotgun_BreakBarrel_Dragon_Athena_VR
    fireRate=0.1
    intialDelay=0
    loop Count=2
    loop Delay=1
    numberOfRoundsPerLoop=3
    weaponName=Pump Shotgun
   }
   {
    1P_WeaponSound=/Game/QA/Art_QA/BR_Audio/Audio/MSS_BR_Audio_Paprika_Sniper_Fire_1p
    3P_WeaponSound=/Game/QA/Art_QA/BR_Audio/Audio/MSS_BR_Audio_Paprika_Sniper_Fire_3p
    ammoCount=30
    associatedItemDefinition=/Game/Athena/Items/Weapons/WID_Sniper_Weather_Athena_SR
    fireRate=0.1
    intialDelay=0
    loop Count=2
    loop Delay=1
    numberOfRoundsPerLoop=3
    weaponName=Sniper
   }
  ]
 }
]
widgets {
 info {rootWidgetClass=&CanvasPanel, widgetCount=15}
 widgets [
  CanvasPanel_20 {widgetClassPath=&CanvasPanel, widget=CanvasPanel_20}
  UIKitGridPanel_0 {
   widgetClassPath=/Script/FortniteUI.UIKitGridPanel
   parent=CanvasPanel_20
   widget=UIKitGridPanel_0
   slot {class=CanvasPanelSlot, properties=~}
  }
  AddButton {widgetClassPath=&EditorUtilityButton, parent=UIKitGridPanel_0, widget=AddButton, slot {class=GridSlot}}
  UEFN_TextBlock_C_39 {
   widgetClassPath=&UEFN_TextBlock
   parent=AddButton
   widget=UEFN_TextBlock_C_39
   properties {
    text=Add Weapon
    font {
     bForceMonospaced=false
     bMaterialIsStencil=false
     fontMaterial=null
     fontObject=/LauncherSocial/BrutalType
     letterSpacing=0
     monospacedWidth=1
     size=18.75
     skewAmount=0
     typefaceFontName=Regular
     outlineSettings {
      bApplyOutlineToDropShadows=false
      bMiteredCorners=false
      bSeparateFillAlpha=false
      outlineMaterial=null
      outlineSize=0
      outlineColor=<rgba 0 0 0 1>
     }
    }
   }
   slot {class=ButtonSlot}
  }
  ClearButton {
   widgetClassPath=&EditorUtilityButton
   parent=UIKitGridPanel_0
   widget=ClearButton
   slot {class=GridSlot, properties {column=1}}
  }
  UEFN_TextBlock_C {
   widgetClassPath=&UEFN_TextBlock
   parent=ClearButton
   widget=UEFN_TextBlock_C
   properties {
    text=Clear Weapons
    font {
     bForceMonospaced=false
     bMaterialIsStencil=false
     fontMaterial=null
     fontObject=&BrutalType
     letterSpacing=0
     monospacedWidth=1
     size=18.75
     skewAmount=0
     typefaceFontName=Regular
     outlineSettings {
      bApplyOutlineToDropShadows=false
      bMiteredCorners=false
      bSeparateFillAlpha=false
      outlineMaterial=null
      outlineSize=0
      outlineColor=<rgba 0 0 0 1>
     }
    }
   }
   slot {class=ButtonSlot}
  }
  AddButton_StopAll {
   widgetClassPath=&EditorUtilityButton
   parent=UIKitGridPanel_0
   widget=AddButton_StopAll
   properties {backgroundColor=<rgba 1 0 0.1284 1>}
   slot {class=GridSlot, properties {column=1, row=1}}
  }
  UEFN_TextBlock_C_3 {
   widgetClassPath=&UEFN_TextBlock
   parent=AddButton_StopAll
   widget=UEFN_TextBlock_C_3
   properties {
    text=Stop All
    font {
     bForceMonospaced=false
     bMaterialIsStencil=false
     fontMaterial=null
     fontObject=&BrutalType
     letterSpacing=0
     monospacedWidth=1
     size=18.75
     skewAmount=0
     typefaceFontName=Regular
     outlineSettings {
      bApplyOutlineToDropShadows=false
      bMiteredCorners=false
      bSeparateFillAlpha=false
      outlineMaterial=null
      outlineSize=0
      outlineColor=<rgba 0 0 0 1>
     }
    }
   }
   slot {class=ButtonSlot}
  }
  AddButton_PlayAll {
   widgetClassPath=&EditorUtilityButton
   parent=UIKitGridPanel_0
   widget=AddButton_PlayAll
   properties {backgroundColor=<rgba 0 1 0.2441 1>}
   slot {class=GridSlot, properties {row=1}}
  }
  UEFN_TextBlock_C_2 {
   widgetClassPath=&UEFN_TextBlock
   parent=AddButton_PlayAll
   widget=UEFN_TextBlock_C_2
   properties {
    text=Play All
    font {
     bForceMonospaced=false
     bMaterialIsStencil=false
     fontMaterial=null
     fontObject=&BrutalType
     letterSpacing=0
     monospacedWidth=1
     size=18.75
     skewAmount=0
     typefaceFontName=Regular
     outlineSettings {
      bApplyOutlineToDropShadows=false
      bMiteredCorners=false
      bSeparateFillAlpha=false
      outlineMaterial=null
      outlineSize=0
      outlineColor=<rgba 0 0 0 1>
     }
    }
   }
   slot {class=ButtonSlot}
  }
  AddButton_AddPaprikaWeapons {
   widgetClassPath=&EditorUtilityButton
   parent=UIKitGridPanel_0
   widget=AddButton_AddPaprikaWeapons
   slot {class=GridSlot, properties {column=2}}
  }
  UEFN_TextBlock_C_1 {
   widgetClassPath=&UEFN_TextBlock
   parent=AddButton_AddPaprikaWeapons
   widget=UEFN_TextBlock_C_1
   properties {
    text=Add Paprika Weapons
    font {
     bForceMonospaced=false
     bMaterialIsStencil=false
     fontMaterial=null
     fontObject=&BrutalType
     letterSpacing=0
     monospacedWidth=1
     size=18.75
     skewAmount=0
     typefaceFontName=Regular
     outlineSettings {
      bApplyOutlineToDropShadows=false
      bMiteredCorners=false
      bSeparateFillAlpha=false
      outlineMaterial=null
      outlineSize=0
      outlineColor=<rgba 0 0 0 1>
     }
    }
   }
   slot {class=ButtonSlot}
  }
  EditorUtilityComboBoxString_0 {
   widgetClassPath=/Script/Blutility.EditorUtilityComboBoxString
   bIsVariable=true
   parent=UIKitGridPanel_0
   widget=EditorUtilityComboBoxString_0
   properties {
    selectedOption=SCAR
    defaultOptions [Firefight]
    foregroundColor {colorUseRule=UseColor_Specified, specifiedColor=<rgba 0.5271 0.5271 0.5271 1>}
   }
   slot {class=GridSlot, properties {column=2, row=1}}
  }
  AudioButtonToggle_DefaultMix {
   widgetClassPath=/AudioWidgets/AudioButtonToggle/AudioButtonToggle
   bIsVariable=true
   parent=UIKitGridPanel_0
   widget=AudioButtonToggle_DefaultMix
   properties {textValue="Default Mix (wip)", visibility=Hidden}
   slot {class=GridSlot, properties {column=3, rowSpan=2}}
  }
  MainBox {
   widgetClassPath=&EditorUtilityScrollBox
   bIsVariable=true
   parent=CanvasPanel_20
   widget=MainBox
   properties {
    alwaysShowScrollbar=true
    alwaysShowScrollbarTrack=true
    consumeMouseWheel=Always
    navigationScrollPadding=-0.016
    scrollbarPadding=<ltrb 10 10 10 10>
   }
   slot {
    class=CanvasPanelSlot
    properties {layoutData {alignment=<0 0>, anchors {maximum=<0 0>, minimum=<0 0>}, offsets=<ltrb 0 120.0003 1996.1818 958.3569>}}
   }
  }
 ]
}
